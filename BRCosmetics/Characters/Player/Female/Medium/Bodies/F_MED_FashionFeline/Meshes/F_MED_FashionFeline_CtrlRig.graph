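# Control Rig for the FashionFeline medium female body that procedurally adjusts deformation and costume bones: skirt, belt, bag, tie and small accessories.
# A shared DeformRig function runs first, then sequenced arm-twist, skirt, belt/bag and tie/misc sections.
# Weighted constraints place costume bones, then swing-twist, cone and spherical pose readers on limbs and neck weight corrective offsets.
# Shoulder and breast tie-collision bones update only when isRigidBodiesEnabled is set.
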ControlRigBlueprint /BRCosmetics/Characters/Player/Female/Medium/Bodies/F_MED_FashionFeline/Meshes/F_MED_FashionFeline_CtrlRig
graphs [
 CollapseNode_ContainedGraph {nodes [
   n1 = RigVMFunctionEntryNode()
   &RigUnit_RotationConstraintLocalSpaceOffset(
    AdvancedSettings=(InterpolationType=Average,RotationOrderForFilter=XZY)
    Child=(Type=Bone,Name="kin_elbowPad_l")
    Filter=(bX=True,bY=True,bZ=True)
    Parents=((Item=(Type=Bone,Name="upperarm_l"),Weight=0.4),(Item=(Type=Bone,Name="lowerarm_l"),Weight=0.6))
    Weight=1
    ExecutePin=@n1.ExecutePin
   )
   > PoseSpaceDeformation = RigVMFunctionReferenceNode(
    Child=(Type=Bone,Name="lowerarm_l")
    Parent=(Type=Null,Name="lowerarm_l_null")
    coneAngle=80
    drawLines=false
    primaryAxis=(X=10,Y=0,Z=0)
    quotient=0.6
    targets=((X=0,Y=0,Z=0.766,W=0.6428),(X=0,Y=0,Z=-0.4226,W=0.9063))
   )
   > &RigUnit_ModifyTransforms(
    ItemToModify=((Item=(Type=Bone,Name="kin_elbowPad_l"),Transform=(Rotation=(X=0,Y=0,Z=-0.0872,W=0.9962),Translation=(X=-2,Y=1,Z=0),Scale3D=(X=1,Y=1,Z=1))))
    WeightMinimum=0
    Weight=RigVMDispatchNode(Index=0, Array=@PoseSpaceDeformation.weights) {template="DISPATCH_RigVMDispatch_ArrayGetAtIndex(in Array,in Index,out Element)"}.Element
   )
   > &RigUnit_ModifyTransforms(
    ItemToModify=((Item=(Type=Bone,Name="kin_elbowPad_l"),Transform=(Rotation=(X=0,Y=0,Z=0.0872,W=0.9962),Translation=(X=2.5,Y=-2,Z=0),Scale3D=(X=1,Y=1,Z=1))))
    WeightMinimum=0
    Weight=RigVMDispatchNode(Index=1, Array=@PoseSpaceDeformation.weights) {template="DISPATCH_RigVMDispatch_ArrayGetAtIndex(in Array,in Index,out Element)"}.Element
   )
   > DeltaQuatExtraction = RigVMFunctionReferenceNode(child=(Type=Bone,Name="hand_r"), onlyTwist=true, parent=(Type=Null,Name="hand_r_null"), twistAxis=(X=1,Y=0,Z=0))
   Interpolate = /Script/RigVM.RigVMFunction_MathQuaternionSlerp(B=(X=0,Y=0,Z=0,W=1), T=0.2, A=@DeltaQuatExtraction.outputQuat) {is=struct RigVMUnitNode, template="Interpolate::Execute(in A,in B,in T,out Result)"}
   > &RigUnit_ModifyTransforms(
    ItemToModify=((Item=(Type=Bone,Name="base_bracelet_r"),Transform=(Rotation=(X=0,Y=0,Z=0,W=1),Translation=(X=0,Y=0,Z=0),Scale3D=(X=1,Y=1,Z=1))))
    Weight=1
    WeightMinimum=0
    ItemToModify.0.Transform.Rotation=@Interpolate.Result
   )
   > &RigUnit_ParentConstraint(
    AdvancedSettings=(RotationOrderForFilter=XZY)
    Child=(Type=Bone,Name="kin_knifeHolster")
    Parents=((Item=(Type=Bone,Name="thigh_twist_01_l"),Weight=0.4399),(Item=(Type=Bone,Name="thigh_l"),Weight=0.5601))
    Weight=1
   )
   > &RigUnit_ParentConstraint(
    AdvancedSettings=(RotationOrderForFilter=XZY)
    Child=(Type=Bone,Name="base_bootZipper_l")
    Parents=((Item=(Type=Bone,Name="calf_twist_02_l"),Weight=0.2189),(Item=(Type=Bone,Name="calf_twist_01_l"),Weight=0.7811))
    Weight=1
   )
   > n2 = &RigUnit_ParentConstraint(
    AdvancedSettings=(RotationOrderForFilter=XZY)
    Child=(Type=Bone,Name="base_bootZipper_r")
    Parents=((Item=(Type=Bone,Name="calf_twist_02_r"),Weight=0.2189),(Item=(Type=Bone,Name="calf_twist_01_r"),Weight=0.7811))
    Weight=1
   )
   tieCollisions = RigVMCollapseNode(ExecutePin=@n2.ExecutePin)
   RigVMFunctionReturnNode(ExecutePin=@tieCollisions.ExecutePin)
  ]}
 CollapseNode_ContainedGraph {nodes [
   n1 = RigVMFunctionEntryNode()
   RigVMFunction_ControlFlowBranch = &RigVMFunction_ControlFlowBranch(ExecuteContext=@n1.ExecutePin, Condition=RigVMVariableNode() {variable=pub isRigidBodiesEnabled}.Value)
   RigVMFunctionReturnNode(ExecutePin=@RigVMFunction_ControlFlowBranch.Completed)
   RigVMFunction_Sequence = RigVMAggregateNode(ExecuteContext=@RigVMFunction_ControlFlowBranch.True)
   &RigUnit_ParentConstraint(
    AdvancedSettings=(RotationOrderForFilter=XZY)
    Child=(Type=Bone,Name="coll_shoulder_l")
    Filter=(TranslationFilter=(bX=True,bY=True,bZ=True),RotationFilter=(bX=False,bY=False,bZ=False),ScaleFilter=(bX=False,bY=False,bZ=False))
    Parents=((Item=(Type=Bone,Name="clavicle_l"),Weight=0.0177),(Item=(Type=Bone,Name="spine_04"),Weight=0.0514),(Item=(Type=Bone,Name="spine_05"),Weight=0.9309))
    Weight=1
    ExecutePin=@RigVMFunction_Sequence.A
   )
   > &RigUnit_ParentConstraint(
    AdvancedSettings=(RotationOrderForFilter=XZY)
    Child=(Type=Bone,Name="coll_shoulder_r")
    Filter=(TranslationFilter=(bX=True,bY=True,bZ=True),RotationFilter=(bX=False,bY=False,bZ=False),ScaleFilter=(bX=False,bY=False,bZ=False))
    Parents=((Item=(Type=Bone,Name="clavicle_r"),Weight=0.0177),(Item=(Type=Bone,Name="spine_04"),Weight=0.0515),(Item=(Type=Bone,Name="spine_05"),Weight=0.9308))
    Weight=1
   )
   > &RigUnit_ParentConstraint(
    AdvancedSettings=(RotationOrderForFilter=XZY)
    Child=(Type=Bone,Name="coll_shoulder_l_target")
    Filter=(TranslationFilter=(bX=True,bY=True,bZ=True),RotationFilter=(bX=False,bY=False,bZ=False),ScaleFilter=(bX=False,bY=False,bZ=False))
    Parents="((Item=(Type=Bone,Name=\"upperarm_twist_01_l\"),Weight=0.0103),(Item=(Type=Bone,Name=\"spine_04\"),Weight=0.0205),(Item=(Type=Bone,Name=\"deform_pec_l\"),Weight=0.0494),(Item=(Type=Bone,Name=\"clavicle_l\"),Weight=0.1906),(Item=(Type=Bone,Name=\"spine_05\"),Weight=0 ...(truncated)"
    Weight=1
   )
   > &RigUnit_ParentConstraint(
    AdvancedSettings=(RotationOrderForFilter=XZY)
    Child=(Type=Bone,Name="coll_shoulder_r_target")
    Filter=(TranslationFilter=(bX=True,bY=True,bZ=True),RotationFilter=(bX=False,bY=False,bZ=False),ScaleFilter=(bX=False,bY=False,bZ=False))
    Parents="((Item=(Type=Bone,Name=\"upperarm_twist_01_r\"),Weight=0.0103),(Item=(Type=Bone,Name=\"spine_04\"),Weight=0.0205),(Item=(Type=Bone,Name=\"deform_pec_r\"),Weight=0.0494),(Item=(Type=Bone,Name=\"clavicle_r\"),Weight=0.1906),(Item=(Type=Bone,Name=\"spine_05\"),Weight=0 ...(truncated)"
    Weight=1
   )
   > &RigUnit_AimConstraintLocalSpaceOffset(
    Child=(Type=Bone,Name="coll_shoulder_l")
    Filter=(bX=True,bY=True,bZ=True)
    Parents=((Item=(Type=Bone,Name="coll_shoulder_l_target"),Weight=1))
    UpAxis=(X=0,Y=0,Z=-1)
    Weight=1
    WorldUp=(Target=(X=0,Y=0,Z=1),Kind=Location,Space=(Type=Bone,Name="spine_05"))
    bMaintainOffset=true
   )
   > &RigUnit_AimConstraintLocalSpaceOffset(
    Child=(Type=Bone,Name="coll_shoulder_r")
    Filter=(bX=True,bY=True,bZ=True)
    Parents=((Item=(Type=Bone,Name="coll_shoulder_r_target"),Weight=1))
    UpAxis=(X=0,Y=0,Z=-1)
    Weight=1
    WorldUp=(Target=(X=0,Y=0,Z=1),Kind=Location,Space=(Type=Bone,Name="spine_05"))
    bMaintainOffset=true
   )
   &RigUnit_ParentConstraint(
    AdvancedSettings=(RotationOrderForFilter=XZY)
    Child=(Type=Bone,Name="coll_breast_l")
    Filter=(TranslationFilter=(bX=True,bY=True,bZ=True),RotationFilter=(bX=False,bY=False,bZ=False),ScaleFilter=(bX=False,bY=False,bZ=False))
    Parents=((Item=(Type=Bone,Name="spine_04"),Weight=0.0626),(Item=(Type=Bone,Name="spine_05"),Weight=0.9374))
    Weight=1
    ExecutePin=@RigVMFunction_Sequence.B
   )
   > &RigUnit_ParentConstraint(
    AdvancedSettings=(RotationOrderForFilter=XZY)
    Child=(Type=Bone,Name="coll_breast_r")
    Filter=(TranslationFilter=(bX=True,bY=True,bZ=True),RotationFilter=(bX=False,bY=False,bZ=False),ScaleFilter=(bX=False,bY=False,bZ=False))
    Parents=((Item=(Type=Bone,Name="deform_pec_r"),Weight=0.032),(Item=(Type=Bone,Name="clavicle_r"),Weight=0.0321),(Item=(Type=Bone,Name="spine_04"),Weight=0.0647),(Item=(Type=Bone,Name="spine_05"),Weight=0.8713))
    Weight=1
   )
   > &RigUnit_ParentConstraint(
    AdvancedSettings=(RotationOrderForFilter=XZY)
    Child=(Type=Bone,Name="coll_breast_l_target")
    Filter=(TranslationFilter=(bX=True,bY=True,bZ=True),RotationFilter=(bX=False,bY=False,bZ=False),ScaleFilter=(bX=False,bY=False,bZ=False))
    Parents=((Item=(Type=Bone,Name="spine_04"),Weight=0.0658),(Item=(Type=Bone,Name="clavicle_l"),Weight=0.0794),(Item=(Type=Bone,Name="deform_pec_l"),Weight=0.0971),(Item=(Type=Bone,Name="spine_05"),Weight=0.7576))
    Weight=1
   )
   > &RigUnit_ParentConstraint(
    AdvancedSettings=(RotationOrderForFilter=XZY)
    Child=(Type=Bone,Name="coll_breast_r_target")
    Filter=(TranslationFilter=(bX=True,bY=True,bZ=True),RotationFilter=(bX=False,bY=False,bZ=False),ScaleFilter=(bX=False,bY=False,bZ=False))
    Parents=((Item=(Type=Bone,Name="clavicle_r"),Weight=0.0517),(Item=(Type=Bone,Name="spine_04"),Weight=0.1876),(Item=(Type=Bone,Name="deform_pec_r"),Weight=0.2172),(Item=(Type=Bone,Name="spine_05"),Weight=0.5435))
    Weight=1
   )
   > &RigUnit_AimConstraintLocalSpaceOffset(
    Child=(Type=Bone,Name="coll_breast_l")
    Filter=(bX=True,bY=True,bZ=True)
    Parents=((Item=(Type=Bone,Name="coll_breast_l_target"),Weight=1))
    UpAxis=(X=0,Y=0,Z=-1)
    Weight=1
    WorldUp=(Target=(X=0,Y=0,Z=1),Kind=Location,Space=(Type=Bone,Name="spine_05"))
    bMaintainOffset=true
   )
   > &RigUnit_AimConstraintLocalSpaceOffset(
    Child=(Type=Bone,Name="coll_breast_r")
    Filter=(bX=True,bY=True,bZ=True)
    Parents=((Item=(Type=Bone,Name="coll_breast_r_target"),Weight=1))
    UpAxis=(X=0,Y=0,Z=-1)
    Weight=1
    WorldUp=(Target=(X=0,Y=0,Z=1),Kind=Location,Space=(Type=Bone,Name="spine_05"))
    bMaintainOffset=true
   )
  ]}
 CollapseNode_ContainedGraph {nodes [
   n1 = RigVMFunctionEntryNode()
   RigVMFunction_MathQuaternionSwingTwist = &RigVMFunction_MathQuaternionSwingTwist(
    TwistAxis=(X=1,Y=0,Z=0)
    Input=&RigUnit_GetRelativeTransformForItem(Child=(Type=Bone,Name="thigh_l"), Parent=(Type=Null,Name="thigh_l_null"), bChildInitial=False, bParentInitial=False).RelativeTransform.Rotation
   )
   RigVMFunction_MathQuaternionToEuler = &RigVMFunction_MathQuaternionToEuler(RotationOrder=ZYX, Value=@RigVMFunction_MathQuaternionSwingTwist.Swing)
   RigVMVariableNode(ExecuteContext=@n1.ExecutePin, Value=@RigVMFunction_MathQuaternionToEuler.Result.Y) {variable=l thighLocal ry float}
   > RigVMVariableNode(Value=@RigVMFunction_MathQuaternionToEuler.Result.Z) {variable=l thighLocal rz float}
   n2 = &RigVMFunction_MathQuaternionToEuler(
    RotationOrder=ZYX
    Value=&RigVMFunction_MathQuaternionSwingTwist(
     TwistAxis=(X=1,Y=0,Z=0)
     Input=&RigUnit_GetRelativeTransformForItem(Child=(Type=Bone,Name="thigh_r"), Parent=(Type=Null,Name="thigh_r_null"), bChildInitial=False, bParentInitial=False).RelativeTransform.Rotation
    ).Swing
   )
   > RigVMVariableNode(Value=@n2.Result.Y) {variable=r thighLocal ry float}
   > n3 = RigVMVariableNode(Value=@n2.Result.Z) {variable=r thighLocal rz float}
   constraints = RigVMCollapseNode(ExecutePin=@n3.ExecuteContext)
   RigVMFunction_Sequence = RigVMAggregateNode(ExecuteContext=@constraints.ExecutePin)
   n4 = &RigVMFunction_MathDoubleRemap(SourceMaximum=-55, SourceMinimum=0, TargetMaximum=1, TargetMinimum=0, Value=RigVMVariableNode() {variable=l thighLocal ry float}.Value)
   &RigUnit_ModifyTransforms(
    ItemToModify="((Item=(Type=Bone,Name=\"base_skirt_E_1\"),Transform=(Rotation=(X=0.0279,Y=-0.3146,Z=-0.0624,W=0.9467),Translation=(X=-0.4441,Y=0,Z=0.7127),Scale3D=(X=1,Y=1,Z=1))),(Item=(Type=Bone,Name=\"base_skirt_C_1\"),Transform=(Rotation=(X=-0.0428,Y=-0.3092,Z=0.1506,W=0. ...(truncated)"
    WeightMinimum=0
    ExecutePin=@RigVMFunction_Sequence.B
    Weight=@n4.Result
   )
   > &RigUnit_ModifyTransforms(
    ItemToModify=((Item=(Type=Bone,Name="kin_knife"),Transform=(Rotation=(X=0,Y=0,Z=0,W=1),Translation=(X=1.9335,Y=0,Z=0),Scale3D=(X=0.9,Y=0.9,Z=0.9))))
    WeightMinimum=0
    Weight=@n4.Result
   )
   > n5 = &RigUnit_SphericalPoseReader(
    ActiveRegionScaleFactors=(PositiveWidth=1,NegativeWidth=1,PositiveHeight=1,NegativeHeight=1)
    ActiveRegionSize=0.05
    Debug=(bDrawDebug=false,bDraw2D=False,bDrawLocalAxes=False,DebugScale=25,DebugSegments=20,DebugThickness=0.25)
    DriverAxis=(X=-1,Y=0,Z=0)
    DriverItem=(Type=Bone,Name="thigh_l")
    FalloffRegionScaleFactors=(PositiveWidth=1,NegativeWidth=1,PositiveHeight=1,NegativeHeight=0.5)
    FalloffSize=0.25
    FlipHeightScaling=False
    FlipWidthScaling=False
    OptionalParentItem=(Type=Bone,Name="None")
    RotationOffset=(X=118,Y=10,Z=0)
   )
   > &RigUnit_ModifyTransforms(
    ItemToModify=((Item=(Type=Bone,Name="base_skirt_B_1"),Transform=(Rotation=(X=-0.0145,Y=-0.085,Z=-0.167,W=0.9822),Translation=(X=0,Y=0,Z=2.1314),Scale3D=(X=1,Y=1,Z=1))))
    WeightMinimum=0
    Weight=@n5.OutputParam
   )
   RigVMFunctionReturnNode(ExecutePin=@RigVMFunction_Sequence.D)
   Remap = &RigVMFunction_MathDoubleRemap(SourceMaximum=-100, SourceMinimum=-5, TargetMaximum=1, TargetMinimum=0, Value=RigVMVariableNode() {variable=l thighLocal rz float}.Value)
   &RigUnit_ModifyTransforms(
    ItemToModify="((Item=(Type=Bone,Name=\"base_skirt_A_1\"),Transform=(Rotation=(X=-0.1381,Y=-0.7405,Z=-0.19,W=0.6297),Translation=(X=-2.2443,Y=0.0001,Z=0.3015),Scale3D=(X=1,Y=1,Z=1))),(Item=(Type=Bone,Name=\"base_skirt_B_1\"),Transform=(Rotation=(X=0.0442,Y=-0.6215,Z=0.0351,W ...(truncated)"
    WeightMinimum=0
    Weight=@Remap.Result
    ExecutePin=@RigVMFunction_Sequence.A
   )
   n6 = &RigVMFunction_MathDoubleRemap(SourceMaximum=-100, SourceMinimum=0, TargetMaximum=1, TargetMinimum=0, Value=RigVMVariableNode() {variable=r thighLocal rz float}.Value)
   > &RigUnit_ModifyTransforms(
    ItemToModify="((Item=(Type=Bone,Name=\"base_skirt_A_1\"),Transform=(Rotation=(X=-0.4438,Y=0.6062,Z=-0.0104,W=-0.6598),Translation=(X=-1.5998,Y=-0.8195,Z=-0.2094),Scale3D=(X=1,Y=1,Z=1))),(Item=(Type=Bone,Name=\"base_skirt_B_1\"),Transform=(Rotation=(X=0.2674,Y=-0.544,Z=0.072 ...(truncated)"
    WeightMinimum=0
    Weight=@n6.Result
   )
   Multiply = &RigVMFunction_MathDoubleMul(B=@Remap.Result, A=@n6.Result)
   > &RigUnit_ModifyTransforms(
    ItemToModify="((Item=(Type=Bone,Name=\"base_skirt_A_1\"),Transform=(Rotation=(X=-0.3187,Y=0.6282,Z=-0.1724,W=0.6885),Translation=(X=0,Y=0,Z=0),Scale3D=(X=1,Y=1,Z=1))),(Item=(Type=Bone,Name=\"base_skirt_B_1\"),Transform=(Rotation=(X=-0.2477,Y=0.6386,Z=-0.0364,W=0.7277),Trans ...(truncated)"
    WeightMinimum=0
    Weight=@Multiply.Result
   )
   &RigUnit_ModifyTransforms(
    ItemToModify="((Item=(Type=Bone,Name=\"base_skirt_D_1\"),Transform=(Rotation=(X=0,Y=-0.4226,Z=0,W=0.9063),Translation=(X=0,Y=0,Z=0),Scale3D=(X=1,Y=1,Z=1))),(Item=(Type=Bone,Name=\"base_skirt_C_1\"),Transform=(Rotation=(X=0,Y=-0.1045,Z=0,W=0.9945),Translation=(X=0,Y=0,Z=0),S ...(truncated)"
    WeightMinimum=0
    Weight=&RigVMFunction_MathDoubleRemap(SourceMaximum=50, SourceMinimum=0, TargetMaximum=1, TargetMinimum=0, Value=RigVMVariableNode() {variable=l thighLocal rz float}.Value).Result
    ExecutePin=@RigVMFunction_Sequence.C
   )
   &RigUnit_ModifyTransforms(
    ExecutePin=()
    ItemToModify="((Item=(Type=Bone,Name=\"base_skirt_A_1\"),Transform=(Rotation=(X=-0.1381,Y=-0.7405,Z=-0.19,W=0.6297),Translation=(X=-2.2443,Y=0.0001,Z=0.3015),Scale3D=(X=1,Y=1,Z=1))),(Item=(Type=Bone,Name=\"base_skirt_B_1\"),Transform=(Rotation=(X=0.0418,Y=-0.6697,Z=0.0378,W ...(truncated)"
    Weight=1
    WeightMinimum=0
   )
  ]}
 CollapseNode_ContainedGraph {nodes [
   n1 = RigVMFunctionEntryNode()
   &RigUnit_ParentConstraint(
    AdvancedSettings=(RotationOrderForFilter=XZY)
    Child=(Type=Bone,Name="base_skirt_A_1")
    Parents="((Item=(Type=Bone,Name=\"deform_groin_l\"),Weight=0.0184),(Item=(Type=Bone,Name=\"deform_groin_r\"),Weight=0.0184),(Item=(Type=Bone,Name=\"thigh_twist_01_l\"),Weight=0.0438),(Item=(Type=Bone,Name=\"thigh_twist_01_r\"),Weight=0.0438),(Item=(Type=Bone,Name=\"spine_01 ...(truncated)"
    Weight=1
    ExecutePin=@n1.ExecutePin
   )
   > &RigUnit_ParentConstraint(
    AdvancedSettings=(RotationOrderForFilter=XZY)
    Child=(Type=Bone,Name="base_skirt_B_1")
    Parents=((Item=(Type=Bone,Name="spine_01"),Weight=0.1128),(Item=(Type=Bone,Name="thigh_twist_01_l"),Weight=0.1279),(Item=(Type=Bone,Name="deform_groin_l"),Weight=0.2058),(Item=(Type=Bone,Name="pelvis"),Weight=0.5535))
    Weight=1
   )
   > &RigUnit_ParentConstraint(
    AdvancedSettings=(RotationOrderForFilter=XZY)
    Child=(Type=Bone,Name="base_skirt_C_1")
    Parents=((Item=(Type=Bone,Name="spine_01"),Weight=0.0532),(Item=(Type=Bone,Name="deform_glute_l"),Weight=0.1341),(Item=(Type=Bone,Name="thigh_twist_01_l"),Weight=0.2738),(Item=(Type=Bone,Name="pelvis"),Weight=0.5389))
    Weight=1
   )
   > &RigUnit_ParentConstraint(
    AdvancedSettings=(RotationOrderForFilter=XZY)
    Child=(Type=Bone,Name="base_skirt_D_1")
    Parents=((Item=(Type=Bone,Name="thigh_twist_01_l"),Weight=0.0556),(Item=(Type=Bone,Name="deform_glute_l"),Weight=0.1145),(Item=(Type=Bone,Name="spine_01"),Weight=0.1305),(Item=(Type=Bone,Name="pelvis"),Weight=0.6993))
    Weight=1
   )
   > n2 = &RigUnit_ParentConstraint(
    AdvancedSettings=(RotationOrderForFilter=XZY)
    Child=(Type=Bone,Name="base_skirt_E_1")
    Parents=((Item=(Type=Bone,Name="spine_01"),Weight=0.0974),(Item=(Type=Bone,Name="deform_groin_l"),Weight=0.1276),(Item=(Type=Bone,Name="thigh_twist_01_l"),Weight=0.2743),(Item=(Type=Bone,Name="pelvis"),Weight=0.5007))
    Weight=1
   )
   RigVMFunctionReturnNode(ExecutePin=@n2.ExecutePin)
  ]}
 CollapseNode_ContainedGraph {nodes [
   n1 = RigVMFunctionEntryNode()
   constraints = RigVMCollapseNode(ExecutePin=@n1.ExecutePin)
   RigVMFunction_Sequence = &RigVMFunction_Sequence(ExecuteContext=@constraints.ExecutePin)
   RigVMFunctionReturnNode(ExecutePin=@RigVMFunction_Sequence.B)
   n2 = &RigVMFunction_MathDoubleRemap(SourceMaximum=-100, SourceMinimum=0, TargetMaximum=1, TargetMinimum=0, Value=RigVMVariableNode() {variable=l thighLocal rz float}.Value)
   &RigUnit_ModifyTransforms(
    ItemToModify="((Item=(Type=Bone,Name=\"kin_belt_1\"),Transform=(Rotation=(X=0,Y=0,Z=0,W=1),Translation=(X=0,Y=-6.6617,Z=3.9193),Scale3D=(X=1,Y=1,Z=1))),(Item=(Type=Bone,Name=\"kin_belt_2\"),Transform=(Rotation=(X=0.258,Y=0,Z=0,W=0.9661),Translation=(X=0,Y=-8.087,Z=5.6512),S ...(truncated)"
    WeightMinimum=0
    Weight=@n2.Result
    ExecutePin=@RigVMFunction_Sequence.A
   )
   Remap = &RigVMFunction_MathDoubleRemap(SourceMaximum=-100, SourceMinimum=0, TargetMaximum=1, TargetMinimum=0, Value=RigVMVariableNode() {variable=r thighLocal rz float}.Value)
   > &RigUnit_ModifyTransforms(
    ItemToModify="((Item=(Type=Bone,Name=\"kin_belt_8\"),Transform=(Rotation=(X=0,Y=0,Z=0,W=1),Translation=(X=1.8451,Y=-0.412,Z=2.563),Scale3D=(X=1,Y=1,Z=1))),(Item=(Type=Bone,Name=\"kin_belt_1\"),Transform=(Rotation=(X=-0.0959,Y=-0.0119,Z=0.1228,W=0.9877),Translation=(X=1.9119 ...(truncated)"
    WeightMinimum=0
    Weight=@Remap.Result
   )
   Multiply = &RigVMFunction_MathDoubleMul(A=@Remap.Result, B=@n2.Result)
   > &RigUnit_ModifyTransforms(
    ItemToModify="((Item=(Type=Bone,Name=\"kin_belt_8\"),Transform=(Rotation=(X=0,Y=0,Z=0.1037,W=0.9946),Translation=(X=-2.9965,Y=7.7888,Z=0),Scale3D=(X=1,Y=1,Z=1))),(Item=(Type=Bone,Name=\"kin_belt_1\"),Transform=(Rotation=(X=0,Y=0,Z=0,W=1),Translation=(X=0,Y=6.2264,Z=1.7185), ...(truncated)"
    WeightMinimum=0
    Weight=@Multiply.Result
   )
   > &RigUnit_ParentConstraint(
    AdvancedSettings=(RotationOrderForFilter=XZY)
    Child=(Type=Bone,Name="base_beltTip")
    Parents=((Item=(Type=Bone,Name="kin_belt_8"),Weight=0.011),(Item=(Type=Bone,Name="kin_belt_3"),Weight=0.0563),(Item=(Type=Bone,Name="kin_belt_1"),Weight=0.4416),(Item=(Type=Bone,Name="kin_belt_2"),Weight=0.4912))
    Weight=1
   )
   > &RigUnit_ModifyTransforms(
    ItemToModify=((Item=(Type=Bone,Name="base_beltTip"),Transform=(Rotation=(X=-0.0109,Y=-0.0266,Z=0.3798,W=-0.9246),Translation=(X=-0.4153,Y=-0.3023,Z=0),Scale3D=(X=1,Y=1,Z=1))))
    WeightMinimum=0
    Weight=@n2.Result
   )
  ]}
 CollapseNode_ContainedGraph {nodes [
   n1 = RigVMFunctionEntryNode()
   &RigUnit_ParentConstraint(
    AdvancedSettings=(RotationOrderForFilter=XZY)
    Child=(Type=Bone,Name="kin_belt_1")
    Parents="((Item=(Type=Bone,Name=\"spine_02\"),Weight=0.0111),(Item=(Type=Bone,Name=\"thigh_twist_01_l\"),Weight=0.0256),(Item=(Type=Bone,Name=\"thigh_twist_01_r\"),Weight=0.0573),(Item=(Type=Bone,Name=\"deform_groin_r\"),Weight=0.0715),(Item=(Type=Bone,Name=\"spine_01\"),Wei ...(truncated)"
    Weight=1
    ExecutePin=@n1.ExecutePin
   )
   > &RigUnit_ParentConstraint(
    AdvancedSettings=(RotationOrderForFilter=XZY)
    Child=(Type=Bone,Name="kin_belt_2")
    Parents="((Item=(Type=Bone,Name=\"spine_02\"),Weight=0.0295),(Item=(Type=Bone,Name=\"thigh_twist_01_l\"),Weight=0.0465),(Item=(Type=Bone,Name=\"deform_groin_l\"),Weight=0.0716),(Item=(Type=Bone,Name=\"spine_01\"),Weight=0.2686),(Item=(Type=Bone,Name=\"pelvis\"),Weight=0.5838 ...(truncated)"
    Weight=1
   )
   > &RigUnit_ParentConstraint(
    AdvancedSettings=(RotationOrderForFilter=XZY)
    Child=(Type=Bone,Name="kin_belt_3")
    Parents="((Item=(Type=Bone,Name=\"spine_02\"),Weight=0.0472),(Item=(Type=Bone,Name=\"deform_groin_l\"),Weight=0.0509),(Item=(Type=Bone,Name=\"thigh_twist_01_l\"),Weight=0.0772),(Item=(Type=Bone,Name=\"spine_01\"),Weight=0.3366),(Item=(Type=Bone,Name=\"pelvis\"),Weight=0.4881 ...(truncated)"
    Weight=1
   )
   > &RigUnit_ParentConstraint(
    AdvancedSettings=(RotationOrderForFilter=XZY)
    Child=(Type=Bone,Name="kin_belt_4")
    Parents=((Item=(Type=Bone,Name="thigh_twist_01_l"),Weight=0.0159),(Item=(Type=Bone,Name="spine_02"),Weight=0.1143),(Item=(Type=Bone,Name="pelvis"),Weight=0.3483),(Item=(Type=Bone,Name="spine_01"),Weight=0.5215))
    Weight=1
   )
   > &RigUnit_ParentConstraint(
    AdvancedSettings=(RotationOrderForFilter=XZY)
    Child=(Type=Bone,Name="kin_belt_5")
    Parents=((Item=(Type=Bone,Name="spine_02"),Weight=0.1318),(Item=(Type=Bone,Name="pelvis"),Weight=0.3663),(Item=(Type=Bone,Name="spine_01"),Weight=0.5019))
    Weight=1
   )
   > &RigUnit_ParentConstraint(
    AdvancedSettings=(RotationOrderForFilter=XZY)
    Child=(Type=Bone,Name="kin_belt_6")
    Parents="((Item=(Type=Bone,Name=\"spine_02\"),Weight=0.0132),(Item=(Type=Bone,Name=\"deform_glute_r\"),Weight=0.0825),(Item=(Type=Bone,Name=\"thigh_twist_01_r\"),Weight=0.0841),(Item=(Type=Bone,Name=\"spine_01\"),Weight=0.1866),(Item=(Type=Bone,Name=\"pelvis\"),Weight=0.6337 ...(truncated)"
    Weight=1
   )
   > &RigUnit_ParentConstraint(
    AdvancedSettings=(RotationOrderForFilter=XZY)
    Child=(Type=Bone,Name="kin_belt_7")
    Parents=((Item=(Type=Bone,Name="deform_glute_r"),Weight=0.0686),(Item=(Type=Bone,Name="spine_01"),Weight=0.0689),(Item=(Type=Bone,Name="thigh_twist_01_r"),Weight=0.3188),(Item=(Type=Bone,Name="pelvis"),Weight=0.5437))
    Weight=1
   )
   > n2 = &RigUnit_ParentConstraint(
    AdvancedSettings=(RotationOrderForFilter=XZY)
    Child=(Type=Bone,Name="kin_belt_8")
    Parents=((Item=(Type=Bone,Name="spine_01"),Weight=0.0182),(Item=(Type=Bone,Name="deform_groin_r"),Weight=0.2773),(Item=(Type=Bone,Name="thigh_twist_01_r"),Weight=0.346),(Item=(Type=Bone,Name="pelvis"),Weight=0.3585))
    Weight=1
   )
   RigVMFunctionReturnNode(ExecutePin=@n2.ExecutePin)
  ]}
 CollapseNode_ContainedGraph {nodes [
   n1 = RigVMFunctionEntryNode()
   &RigUnit_ParentConstraint(
    AdvancedSettings=(RotationOrderForFilter=XZY)
    Child=(Type=Bone,Name="kin_tieKnot")
    Filter=(TranslationFilter=(bX=True,bY=True,bZ=True),RotationFilter=(bX=True,bY=True,bZ=True),ScaleFilter=(bX=False,bY=False,bZ=False))
    Parents=((Item=(Type=Bone,Name="head"),Weight=0.0842),(Item=(Type=Bone,Name="spine_04"),Weight=0.113),(Item=(Type=Bone,Name="neck_02"),Weight=0.1939),(Item=(Type=Bone,Name="neck_01"),Weight=0.205),(Item=(Type=Bone,Name="spine_05"),Weight=0.4039))
    Weight=1
    ExecutePin=@n1.ExecutePin
   )
   > DeltaRotation v4 = RigVMFunctionReferenceNode(
    Rotation Order=ZYX
    absoluteValue=false
    child=(Type=Bone,Name="neck_02")
    debugRotation=false
    onlyTwist=false
    outputMinMax=(X=0,Y=1)
    parent=(Type=Null,Name="neck_02_null")
    premult=1
    rotationMinMax=(X=0,Y=25)
    triggerAxis=Z
    twistAxis=(X=1,Y=0,Z=0)
   )
   > &RigUnit_ModifyTransforms(
    ItemToModify=((Item=(Type=Bone,Name="kin_tieKnot"),Transform=(Rotation=(X=-0.1736,Y=0,Z=0,W=-0.9848),Translation=(X=0,Y=0.35,Z=0),Scale3D=(X=1,Y=1,Z=1))))
    WeightMinimum=0
    Weight=@"DeltaRotation v4".outputWeight
   )
   > DeltaRotation v4_2 = RigVMFunctionReferenceNode(
    Rotation Order=ZYX
    absoluteValue=false
    child=(Type=Bone,Name="neck_02")
    debugRotation=false
    onlyTwist=false
    outputMinMax=(X=0,Y=1)
    parent=(Type=Null,Name="neck_02_null")
    premult=1
    rotationMinMax=(X=0,Y=-25)
    triggerAxis=Z
    twistAxis=(X=1,Y=0,Z=0)
   )
   > &RigUnit_ModifyTransforms(
    ItemToModify="((Item=(Type=Bone,Name=\"base_tieBot_1_l\"),Transform=(Rotation=(X=0,Y=0,Z=0.1736,W=0.9848),Translation=(X=0,Y=0,Z=0),Scale3D=(X=1,Y=1,Z=1))),(Item=(Type=Bone,Name=\"base_tieBot_1_r\"),Transform=(Rotation=(X=0,Y=0,Z=0.1736,W=0.9848),Translation=(X=0,Y=0,Z=0),S ...(truncated)"
    WeightMinimum=0
    Weight=@"DeltaRotation v4_2".outputWeight
   )
   > DeltaRotation v4_1 = RigVMFunctionReferenceNode(
    Rotation Order=ZYX
    absoluteValue=true
    child=(Type=Bone,Name="neck_02")
    debugRotation=false
    onlyTwist=true
    outputMinMax=(X=-1,Y=1)
    parent=(Type=Null,Name="neck_02_null")
    premult=1
    rotationMinMax=(X=26,Y=-26)
    triggerAxis=X
    twistAxis=(X=1,Y=0,Z=0)
   )
   > &RigUnit_ModifyTransforms(
    ItemToModify=((Item=(Type=Bone,Name="kin_tieKnot"),Transform=(Rotation=(X=0.0436,Y=0.0008,Z=0.0174,W=0.9989),Translation=(X=0,Y=0,Z=0),Scale3D=(X=1,Y=1,Z=1))))
    WeightMinimum=0
    Weight=@"DeltaRotation v4_1".outputWeight
   )
   > n2 = &RigUnit_SphericalPoseReader(
    ActiveRegionScaleFactors=(PositiveWidth=1,NegativeWidth=1,PositiveHeight=1,NegativeHeight=1)
    ActiveRegionSize=0.05
    Debug=(bDrawDebug=false,bDraw2D=False,bDrawLocalAxes=False,DebugScale=25,DebugSegments=20,DebugThickness=0.25)
    DriverAxis=(X=1,Y=0,Z=0)
    DriverItem=(Type=Bone,Name="neck_02")
    FalloffRegionScaleFactors=(PositiveWidth=1,NegativeWidth=1,PositiveHeight=1,NegativeHeight=1)
    FalloffSize=0.1
    FlipHeightScaling=False
    FlipWidthScaling=False
    OptionalParentItem=(Type=Bone,Name="spine_05")
    RotationOffset=(X=-6.2881,Y=-70.4608,Z=-10.2999)
   )
   > n3 = &RigUnit_ModifyTransforms(
    ItemToModify=((Item=(Type=Bone,Name="kin_tieKnot"),Transform=(Rotation=(X=-0.069,Y=-0.0113,Z=0.1032,W=-0.9922),Translation=(X=0,Y=0.7624,Z=0.1323),Scale3D=(X=1,Y=1,Z=1))))
    WeightMinimum=0
    Weight=@n2.OutputParam
   )
   RigVMFunctionReturnNode(ExecutePin=@n3.ExecutePin)
  ]}
 CollapseNode_ContainedGraph {nodes [
   n1 = RigVMFunctionEntryNode()
   &RigUnit_ParentConstraint(
    AdvancedSettings=(RotationOrderForFilter=XZY)
    Child=(Type=Bone,Name="base_bag")
    Parents=((Item=(Type=Bone,Name="kin_belt_6"),Weight=0.106),(Item=(Type=Bone,Name="kin_belt_8"),Weight=0.327),(Item=(Type=Bone,Name="kin_belt_7"),Weight=0.5671))
    Weight=1
    ExecutePin=@n1.ExecutePin
   )
   Remap = &RigVMFunction_MathDoubleRemap(SourceMaximum=-50, SourceMinimum=0, TargetMaximum=1, TargetMinimum=0, Value=RigVMVariableNode() {variable=r thighLocal ry float}.Value)
   > &RigUnit_ModifyTransforms(
    ItemToModify=((Item=(Type=Bone,Name="base_bag"),Transform=(Rotation=(X=0,Y=0,Z=-0.1736,W=0.9848),Translation=(X=1,Y=0,Z=0),Scale3D=(X=1,Y=1,Z=1))))
    WeightMinimum=0
    Weight=@Remap.Result
   )
   > &RigUnit_ModifyTransforms(
    ItemToModify=((Item=(Type=Bone,Name="base_bag"),Transform=(Rotation=(X=-0.003,Y=-0.1736,Z=-0.0172,W=0.9847),Translation=(X=0,Y=0,Z=0),Scale3D=(X=1,Y=1,Z=1))))
    WeightMinimum=0
    Weight=&RigVMFunction_MathDoubleRemap(SourceMaximum=-90, SourceMinimum=0, TargetMaximum=1, TargetMinimum=0, Value=RigVMVariableNode() {variable=r thighLocal rz float}.Value).Result
   )
   > n2 = &RigUnit_SphericalPoseReader(
    ActiveRegionScaleFactors=(PositiveWidth=1,NegativeWidth=1,PositiveHeight=1,NegativeHeight=1)
    ActiveRegionSize=0.05
    Debug=(bDrawDebug=false,bDraw2D=False,bDrawLocalAxes=False,DebugScale=25,DebugSegments=20,DebugThickness=0.25)
    DriverAxis=(X=1,Y=0,Z=0)
    DriverItem=(Type=Bone,Name="thigh_r")
    FalloffRegionScaleFactors=(PositiveWidth=1,NegativeWidth=1,PositiveHeight=1,NegativeHeight=1)
    FalloffSize=0.2
    FlipHeightScaling=False
    FlipWidthScaling=False
    OptionalParentItem=(Type=Bone,Name="None")
    RotationOffset=(X=-60,Y=0,Z=14)
   )
   > n3 = &RigUnit_ModifyTransforms(
    ItemToModify=((Item=(Type=Bone,Name="base_bag"),Transform=(Rotation=(X=-0.0436,Y=0,Z=0,W=0.999),Translation=(X=0,Y=0,Z=0),Scale3D=(X=1,Y=1,Z=1))))
    WeightMinimum=0
    Weight=@n2.OutputParam
   )
   RigVMFunctionReturnNode(ExecutePin=@n3.ExecutePin)
  ]}
 Multiply_3_ContainedGraph {nodes [
   n1 = RigVMFunctionEntryNode()
   RigVMFunctionReturnNode(Result=&RigVMFunction_MathTransformMul(B=@n1.C, A=&RigVMFunction_MathTransformMul(A=@n1.A, B=@n1.B) {template="Multiply::Execute(in A,in B,out Result)"}.Result) {template="Multiply::Execute(in A,in B,out Result)"}.Result)
  ]}
 New Function_ContainedGraph {nodes [
   RigVMFunctionEntryNode()
   > DeltaRotation v4 = RigVMFunctionReferenceNode(
    Rotation Order=ZYX
    child=(Type=Bone,Name="hand_l")
    debugRotation=false
    onlyTwist=True
    outputMinMax=(X=0,Y=1)
    parent=(Type=Null,Name="hand_l_null")
    premult=1
    rotationMinMax=(X=120,Y=150)
    twistAxis=(X=1,Y=0,Z=0)
   )
   > &RigUnit_ModifyTransforms(
    ItemToModify="((Item=(Type=Bone,Name=\"lowerarm_twist_01_l\"),Transform=(Rotation=(X=0,Y=-0.0262,Z=0,W=0.9997),Translation=(X=0,Y=0,Z=1),Scale3D=(X=1,Y=1.2,Z=1.2))),(Item=(Type=Bone,Name=\"lowerarm_twist_02_l\"),Transform=(Rotation=(X=0.2588,Y=0,Z=0,W=0.9659),Translation=(X ...(truncated)"
    WeightMinimum=0
    Weight=@"DeltaRotation v4".outputWeight
   )
   > RigVMFunctionReturnNode()
  ]}
 New Function_ContainedGraph {nodes [
   n1 = RigVMFunctionEntryNode()
   RigVMFunction_MathBoolToInteger = &RigVMFunction_MathBoolToInteger(Value=@n1.onlyTwist)
   RigVMFunction_MathQuaternionSwingTwist = &RigVMFunction_MathQuaternionSwingTwist(
    TwistAxis=@n1.twistAxis
    Input=&RigUnit_GetRelativeTransformForItem(bChildInitial=False, bParentInitial=False, Child=@n1.child, Parent=@n1.parent).RelativeTransform.Rotation
   )
   n2 = RigVMDispatchNode(
    Values=((),())
    Index=@RigVMFunction_MathBoolToInteger.Result
    Values.0=@RigVMFunction_MathQuaternionSwingTwist.Swing
    Values.1=@RigVMFunction_MathQuaternionSwingTwist.Twist
   ) {template="DISPATCH_RigVMDispatch_SelectInt32(in Index,in Values,out Result)"}
   RigVMFunction_MathQuaternionToEuler = &RigVMFunction_MathQuaternionToEuler(Value=@n2.Result, RotationOrder=@n1.Rotation Order)
   > DISPATCH_RigVMDispatch_Print = RigVMDispatchNode(ScreenColor=(R=1,G=1,B=1,A=1), ScreenDuration=1.5, Enabled=@n1.debugRotation, Value=@RigVMFunction_MathQuaternionToEuler.Result) {template="DISPATCH_RigVMDispatch_Print(in Prefix,in Value,in Enabled,in ScreenDuration,in ScreenColor)"}
   Cast = RigVMDispatchNode(Value=@n1.triggerAxis) {template="DISPATCH_RigVMDispatch_CastEnumToInt(in Value,out Result)"}
   DISPATCH_RigVMDispatch_SelectInt32 = RigVMDispatchNode(
    Values=(0,0,0)
    Index=@Cast.Result
    Values.0=@RigVMFunction_MathQuaternionToEuler.Result.X
    Values.1=@RigVMFunction_MathQuaternionToEuler.Result.Y
    Values.2=@RigVMFunction_MathQuaternionToEuler.Result.Z
   ) {template="DISPATCH_RigVMDispatch_SelectInt32(in Index,in Values,out Result)"}
   Remap = &RigVMFunction_MathDoubleRemap(
    bClamp=False
    Value=@DISPATCH_RigVMDispatch_SelectInt32.Result
    TargetMinimum=@n1.outputMinMax.X
    TargetMaximum=@n1.outputMinMax.Y
    SourceMinimum=@n1.rotationMinMax.X
    SourceMaximum=@n1.rotationMinMax.Y
   )
   Absolute = &RigVMFunction_MathDoubleAbs(Value=@Remap.Result)
   Multiply = &RigVMFunction_MathDoubleMul(
    A=&RigVMFunction_MathDoubleClamp(Value=RigVMDispatchNode(True=@Absolute.Result, Condition=@n1.absoluteValue, False=@Remap.Result) {template="DISPATCH_RigVMDispatch_If(in Condition,in True,in False,out Result)"}.Result).Result
    B=@n1.premult
   )
   > RigVMFunctionReturnNode(outputQuat=@n2.Result, outputWeight=@Multiply.Result)
  ]}
 New Function_ContainedGraph {nodes [
   n1 = RigVMFunctionEntryNode()
   RigVMFunction_MathBoolToInteger = &RigVMFunction_MathBoolToInteger(Value=@n1.onlyTwist)
   RigVMFunction_MathQuaternionSwingTwist = &RigVMFunction_MathQuaternionSwingTwist(
    TwistAxis=@n1.twistAxis
    Input=&RigUnit_GetRelativeTransformForItem(bChildInitial=False, bParentInitial=False, Child=@n1.child, Parent=@n1.parent).RelativeTransform.Rotation
   )
   n2 = RigVMDispatchNode(
    Values=((),())
    Index=@RigVMFunction_MathBoolToInteger.Result
    Values.0=@RigVMFunction_MathQuaternionSwingTwist.Swing
    Values.1=@RigVMFunction_MathQuaternionSwingTwist.Twist
   ) {template="DISPATCH_RigVMDispatch_SelectInt32(in Index,in Values,out Result)"}
   Inverse = /Script/RigVM.RigVMFunction_MathQuaternionInverse(Value=@n2.Result) {is=struct RigVMUnitNode, template="Inverse::Execute(in Value,out Result)"}
   > RigVMFunctionReturnNode(outputQuat=RigVMDispatchNode(True=@Inverse.Result, Condition=@n1.invert, False=@n2.Result) {template="DISPATCH_RigVMDispatch_If(in Condition,in True,in False,out Result)"}.Result)
  ]}
 New Function_ContainedGraph {nodes [
   n1 = RigVMFunctionEntryNode()
   Num = RigVMDispatchNode(Array=@n1.targets) {template="DISPATCH_RigVMDispatch_ArrayGetNum(in Array,out Num)"}
   > RigVMDispatchNode(Element=0, Num=@Num.Num, Array=RigVMVariableNode() {variable=outputValues}.Value) {template="DISPATCH_RigVMDispatch_ArrayInit(io Array,in Num,in Element)"}
   n2 = &RigVMFunction_MathTransformMul(
    A=&RigUnit_GetRelativeTransformForItem(bChildInitial=true, bParentInitial=true, Child=@n1.Child, Parent=@n1.Parent).RelativeTransform
    B=&RigUnit_GetTransform(bInitial=False, Item=@n1.Parent).Transform
   ) {template="Multiply::Execute(in A,in B,out Result)"}
   n3 = &RigVMFunction_MathTransformMul(
    A=(Rotation=(X=0,Y=0,Z=0,W=1),Translation=(X=0,Y=0,Z=0),Scale3D=(X=1,Y=1,Z=1))
    A.Translation=@n1.primaryAxis
    B=&RigUnit_GetTransform(bInitial=False, Item=@n1.Child).Transform
   ) {template="Multiply::Execute(in A,in B,out Result)"}
   > /Script/RigVM.RigVMFunction_DebugLineNoSpace(
    Color=(R=0,G=0,B=1,A=1)
    DebugDrawSettings=(DepthPriority=SDPG_Foreground,Lifetime=-1)
    Thickness=0.5
    WorldOffset=(Rotation=(X=0,Y=0,Z=0,W=1),Translation=(X=0,Y=0,Z=0),Scale3D=(X=1,Y=1,Z=1))
    bEnabled=@n1.drawLines
    A=@n2.Result.Translation
    B=@n3.Result.Translation
   ) {is=struct RigVMUnitNode}
   > For_Each = RigVMDispatchNode(Array=@n1.targets) {template="DISPATCH_RigVMDispatch_ArrayIterator(in Array,out Element,out Index,out Count,out Ratio)"}
   Multiply_1 = RigVMAggregateNode(
    A=(Rotation=(X=0,Y=0,Z=0,W=1),Translation=(X=0,Y=0,Z=0),Scale3D=(X=1,Y=1,Z=1))
    B=(Rotation=(X=0,Y=0,Z=0,W=1),Translation=(X=0,Y=0,Z=0),Scale3D=(X=1,Y=1,Z=1))
    A.Translation=@n1.primaryAxis
    B.Rotation=@For_Each.Element
    C=@n2.Result
   )
   > &RigVMFunction_DebugLineNoSpace(
    Color=(R=1,G=0,B=0,A=1)
    DebugDrawSettings=(DepthPriority=SDPG_Foreground,Lifetime=-1)
    Thickness=0.5
    WorldOffset=(Rotation=(X=0,Y=0,Z=0,W=1),Translation=(X=0,Y=0,Z=0),Scale3D=(X=1,Y=1,Z=1))
    bEnabled=@n1.drawLines
    B=@Multiply_1.Result.Translation
    A=@n2.Result.Translation
   )
   RigVMFunction_MathVectorAngle = /Script/RigVM.RigVMFunction_MathVectorAngle(
    B=/Script/RigVM.RigVMFunction_MathVectorSub(A=@Multiply_1.Result.Translation, B=@n2.Result.Translation) {is=struct RigVMUnitNode, template="Subtract::Execute(in A,in B,out Result)"}.Result
    A=&RigVMFunction_MathVectorSub(B=@n2.Result.Translation, A=@n3.Result.Translation) {template="Subtract::Execute(in A,in B,out Result)"}.Result
   ) {is=struct RigVMUnitNode, template="AngleBetween::Execute(in A,in B,out Result)"}
   Degrees = /Script/RigVM.RigVMFunction_MathDoubleDeg(Value=@RigVMFunction_MathVectorAngle.Result) {is=struct RigVMUnitNode, template="Degrees::Execute(in Value,out Result)"}
   Multiply = &RigVMFunction_MathDoubleMul(A=@n1.coneAngle, B=@n1.quotient)
   Remap = &RigVMFunction_MathDoubleRemap(SourceMaximum=0, TargetMaximum=1, TargetMinimum=0, Value=@Degrees.Result, SourceMinimum=@Multiply.Result)
   > Set_At = RigVMDispatchNode(Index=@For_Each.Index, Element=@Remap.Result, Array=RigVMVariableNode() {variable=outputValues}.Value) {template="DISPATCH_RigVMDispatch_ArraySetAtIndex(io Array,in Index,in Element)"}
   RigVMFunctionReturnNode(ExecuteContext=@For_Each.Completed, weights=RigVMVariableNode() {variable=outputValues}.Value)
  ]}
 RigVMFunctionLibrary {nodes [
   DeltaQuatExtraction = RigVMCollapseNode(ExecuteContext=(), child=(Type=Bone,Name="None"), parent=(Type=Bone,Name="None"), twistAxis=(X=1,Y=0,Z=0))
   DeltaRotation v4 = RigVMCollapseNode(
    ExecuteContext=()
    Rotation Order=ZYX
    child=(Type=Bone,Name="None")
    outputMinMax=(X=0,Y=1)
    parent=(Type=Bone,Name="None")
    premult=1
    rotationMinMax=(X=0,Y=0)
    twistAxis=(X=1,Y=0,Z=0)
   )
   PoseSpaceDeformation = RigVMCollapseNode(
    Child=(Type=Bone,Name="None")
    ExecuteContext=()
    Parent=(Type=Bone,Name="None")
    coneAngle=90
    drawLines=true
    primaryAxis=(X=0,Y=0,Z=0)
    quotient=0.5
    targets=()
   )
  ]}
 RigVMFunction_Sequence_1_ContainedGraph {nodes [
   n1 = RigVMFunctionEntryNode()
   RigVMFunction_Sequence = &RigVMFunction_Sequence(ExecuteContext=@n1.ExecuteContext)
   n2 = &RigVMFunction_Sequence(ExecuteContext=@RigVMFunction_Sequence.B)
   RigVMFunctionReturnNode(A=@RigVMFunction_Sequence.A, B=@n2.A, C=@n2.B)
  ]}
 RigVMFunction_Sequence_1_ContainedGraph {nodes [
   n1 = RigVMFunctionEntryNode()
   RigVMFunction_Sequence = &RigVMFunction_Sequence(ExecuteContext=@n1.ExecuteContext)
   n2 = &RigVMFunction_Sequence(ExecuteContext=@RigVMFunction_Sequence.B)
   n3 = &RigVMFunction_Sequence(ExecuteContext=@n2.B)
   RigVMFunctionReturnNode(A=@RigVMFunction_Sequence.A, B=@n2.A, C=@n3.A, D=@n3.B)
  ]}
 RigVMFunction_Sequence_1_ContainedGraph {nodes [
   n1 = RigVMFunctionEntryNode()
   RigVMFunction_Sequence = &RigVMFunction_Sequence(ExecuteContext=@n1.ExecuteContext)
   n2 = &RigVMFunction_Sequence(ExecuteContext=@RigVMFunction_Sequence.B)
   RigVMFunctionReturnNode(A=@RigVMFunction_Sequence.A, B=@n2.A, C=@n2.B)
  ]}
 RigVMModel {nodes [
   RigUnit_BeginExecution = &RigUnit_BeginExecution()
   FN_DeformRig_FNC = RigVMFunctionReferenceNode(
    DeformRig_ Curve="(deformRig_l_upperarm_up_alpha_72_#1=False,deformRig_r_upperarm_up_alpha_73_#2=False,deformRig_l_upperarm_fr_alpha_74_#3=False,deformRig_r_upperarm_fr_alpha_75_CB51B8 ...(truncated)"
    DeformRig_Bool="(upperarm_1_#4=True,elbow_3_#5=True,wrist_7_#6=True,groin_8_#7=true,glute_17_#8=true,knee_9_40C408324CA261 ...(truncated)"
    DeformRig_Debug=(upperarm_debug_18_#4=False,wrist_debug_20_#6=False,lat_debug_26_#9=False,groinGlute_debug_27_#7=False)
    DeformRig_Value="(upperarm_up_18_#10=(Rotation=(X=0,Y=0,Z=0,W=1),Translation=(X=0,Y=0,Z=3),Scale3D=(X=0.7,Y=1,Z=1)),upperarm_fwd_fr_19_#11=(Rotation=(X=0,Y=0,Z=0,W=1),Translation=(X=0,Y=10,Z=0),Scale3D=(X=0.5,Y=1,Z= ...(truncated)"
    ExecuteContext=@RigUnit_BeginExecution.ExecutePin
   )
   > RigVMFunction_Sequence = RigVMAggregateNode()
   armTwistFix = RigVMCollapseNode(ExecuteContext=@RigVMFunction_Sequence.A)
   skirt = RigVMCollapseNode(ExecutePin=@armTwistFix.ExecuteContext)
   belt = RigVMCollapseNode(ExecutePin=@RigVMFunction_Sequence.B)
   bag = RigVMCollapseNode(ExecutePin=@belt.ExecutePin)
   tie = RigVMCollapseNode(ExecutePin=@RigVMFunction_Sequence.C)
   misc = RigVMCollapseNode(ExecutePin=@tie.ExecutePin)
   setFKCtrl = RigVMFunctionReferenceNode(ExecuteContext=(), setFK_Ctrl=false)
  ]}
 RigVMModel Rig {nodes [
   n1 = &RigUnit_PrepareForExecution()
   &RigUnit_HierarchyImportFromSkeleton(NameSpace=None, bIncludeCurves=true, bIncludeMeshSockets=true, bIncludeVirtualBones=True, ExecutePin=@n1.ExecutePin)
   Item Array = RigVMDispatchNode(Value="((Type=Bone,Name=\"thigh_l\"),(Type=Bone,Name=\"calf_l\"),(Type=Bone,Name=\"foot_l\"),(Type=Bone,Name=\"thigh_r\"),(Type=Bone,Name=\"calf_r\"),(Type=Bone,Name=\"foot_r\"),(Type=Bone,Name=\"clavicle_l\"),(Type=Bone,Name=\"upperarm_l\"),(Type=Bone,Name=\"hand_l\"),(Type=Bone, ...(truncated)") {template="DISPATCH_RigVMDispatch_Constant(io Value)"}
   > addNullObject = RigVMFunctionReferenceNode(
    Options_Null="(pelvis_1_#12=False,spine_3_#13=False,clavicle_5_#14=False,arm_7_#15=False,leg_9_#16=False,head_11_46DC15EF465C ...(truncated)"
    Custom_Joints=@"Item Array".Value
   )
   > &RigUnit_HierarchyAddNull(
    Name=neck_02_null
    Parent=(Type=Bone,Name="spine_05")
    Space=LocalSpace
    Transform=&RigUnit_GetRelativeTransformForItem(Child=(Type=Bone,Name="neck_02"), Parent=(Type=Bone,Name="spine_05"), bChildInitial=true, bParentInitial=true).RelativeTransform
   )
   addFKCtrl = RigVMFunctionReferenceNode(
    Custom_FKCtrl=()
    ExecuteContext=()
    Options_FKCtrl="(pelvis_1_#12=true,spine_3_#13=true,clavicle_5_#14=true,arm_7_#15=true,leg_9_#16=true,head_11_46DC15EF465CD2979 ...(truncated)"
    Shape_Settings="(pelvis_2_#17=(bVisible=True,Name=\"Hexagon_Thin\",Color=(R=0.0117,G=0.8281,B=1,A=1),Transform=(Rotation=(X=0,Y=-0.7071,Z=0,W=0.7071),Translation=(X=2,Y=0,Z=0),Scale3D=(X=5,Y=5,Z=7))),spine_4_#18=(bVi ...(truncated)"
   )
  ]}
]
variables [
 l thighLocal ry float {type=double, default=0}
 l thighLocal rz float {type=double, default=0}
 pub isRigidBodiesEnabled {type=bool, default=False}
 r thighLocal ry float {type=double, default=0}
 r thighLocal rz float {type=double, default=0}
]
