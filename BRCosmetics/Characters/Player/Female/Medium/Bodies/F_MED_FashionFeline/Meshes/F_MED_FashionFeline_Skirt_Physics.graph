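# Physics asset for the skirt of the female medium FashionFeline player cosmetic body, defining bodies and joints for its movement and collision.
# Five chains (A–E) run from tiny base capsules to an upper capsule and a lower box, with extra capsules on C and E.
# Tight angular limits join each chain, with one swing axis locked to zero degrees at the top.
# Pelvis, thigh and left-thigh knife capsules provide body shapes, linked by free-rotation constraints.
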PhysicsAsset /BRCosmetics/Characters/Player/Female/Medium/Bodies/F_MED_FashionFeline/Meshes/F_MED_FashionFeline_Skirt_Physics
bodies {
 base_skirt_A_1 [{shapeType=Capsule, length=1, radius=0.5, shapeName=base_skirt_A_1_capsule}]
 base_skirt_B_1 [{shapeType=Capsule, length=1, radius=0.5, shapeName=base_skirt_B_1_capsule}]
 base_skirt_C_1 [{shapeType=Capsule, length=1, radius=0.5, shapeName=base_skirt_C_1_capsule}]
 base_skirt_D_1 [{shapeType=Capsule, length=1, radius=0.5, shapeName=base_skirt_D_1_capsule}]
 base_skirt_E_1 [{shapeType=Capsule, length=1, radius=0.5, shapeName=base_skirt_E_1_capsule}]
 dyn_skirt_A_1 [{shapeType=Capsule, length=10.3221, radius=2.9683, shapeName=dyn_skirt_A_1_capsule, center=<3.832 0 2.1063>, rotation=<rot 0 0 90>}]
 dyn_skirt_A_2 [{shapeType=Box, extentX=11.6713, extentY=9.5775, extentZ=2.1158, shapeName=dyn_skirt_A_2_box, center=<6.6727 0 0>}]
 dyn_skirt_B_1 [{shapeType=Capsule, length=10.3221, radius=2.9683, shapeName=dyn_skirt_A_1_capsule, center=<3.832 0 2.1063>, rotation=<rot 0 0 90>}]
 dyn_skirt_B_2 [{shapeType=Box, extentX=11.6713, extentY=9.5775, extentZ=2.1158, shapeName=dyn_skirt_B_2_box, center=<6.6727 0 0>}]
 dyn_skirt_C_1 [{shapeType=Capsule, length=10.3221, radius=2.9683, shapeName=dyn_skirt_A_1_capsule, center=<3.832 0 2.1063>, rotation=<rot 0 0 90>}]
 dyn_skirt_C_2 [
  {
   shapeType=Capsule
   length=10.4293
   radius=2.3257
   shapeName=dyn_skirt_C_2_capsule
   center=<5.7115 6.1724 -0.1164>
   rotation=<rot -71.7814 -102.802 105.5907>
  }
  {shapeType=Box, extentX=11.6713, extentY=9.5775, extentZ=2.1158, shapeName=dyn_skirt_C_2_box, center=<6.6727 0 0>}
 ]
 dyn_skirt_D_1 [{shapeType=Capsule, length=10.3221, radius=2.9683, shapeName=dyn_skirt_A_1_capsule, center=<3.832 0 1.3886>, rotation=<rot 0 0 90>}]
 dyn_skirt_D_2 [{shapeType=Box, extentX=11.6713, extentY=9.5775, extentZ=2.1158, shapeName=dyn_skirt_D_2_box, center=<6.6727 0 0>}]
 dyn_skirt_E_1 [{shapeType=Capsule, length=10.3221, radius=2.9683, shapeName=dyn_skirt_A_1_capsule, center=<3.832 0 2.1063>, rotation=<rot 0 0 90>}]
 dyn_skirt_E_2 [
  {
   shapeType=Capsule
   length=10.5526
   radius=2.2025
   shapeName=dyn_skirt_C_2_capsule
   center=<6.1917 -6.3786 0.2699>
   rotation=<rot -71.7814 -102.802 105.5907>
  }
  {shapeType=Box, extentX=11.6713, extentY=9.5775, extentZ=2.1158, shapeName=dyn_skirt_E_2_box, center=<6.6727 0 0>}
 ]
 kin_knife [{
   shapeType=Capsule
   length=21.5591
   radius=1.5401
   shapeName=kin_knife_capsule
   center=<2.1085 -0.0405 0.2845>
   rotation=<rot -0.0186 90.9483 -88.8789>
  }]
 pelvis [
  {
   shapeType=Capsule
   length=13.1041
   radius=9.1551
   shapeName=pelvis_capsule
   center=<-10.1716 -1.2328 0.8022>
   rotation=<rot -1.9782 -89.9987 -0.0044>
  }
  {
   shapeType=Capsule
   length=14.7836
   radius=7.4755
   shapeName=pelvis_capsule
   center=<-6.7678 -4.7563 1.9926>
   rotation=<rot -1.9782 -89.9987 -0.0044>
  }
 ]
 thigh_l [
  {shapeType=Capsule, length=11.5973, radius=9.2546, shapeName=thigh_l_capsule, center=<-16.7005 -0.2611 0.9285>, rotation=<rot -90 -3.6255 0>}
  {shapeType=Capsule, length=11.5821, radius=9.0624, shapeName=thigh_l_capsule, center=<-28.713 0.2777 0.4494>, rotation=<rot -90 4.6762 0>}
 ]
 thigh_r [{shapeType=Capsule, length=20.8322, radius=8.6749, shapeName=thigh_l_capsule, center=<20.0405 0.0895 0.6401>, rotation=<rot -90 -3.6255 0>}]
}
constraints [
 {bone1Name=dyn_skirt_A_1, bone2Name=base_skirt_A_1, swing1LimitDegrees=0, swing2LimitDegrees=20, twistLimitDegrees=10}
 {bone1Name=dyn_skirt_A_2, bone2Name=dyn_skirt_A_1, swing1LimitDegrees=10, swing2LimitDegrees=20, twistLimitDegrees=10}
 {bone1Name=dyn_skirt_B_1, bone2Name=base_skirt_B_1, swing1LimitDegrees=0, swing2LimitDegrees=18, twistLimitDegrees=10}
 {bone1Name=dyn_skirt_B_2, bone2Name=dyn_skirt_B_1, swing1LimitDegrees=10, swing2LimitDegrees=20, twistLimitDegrees=10}
 {bone1Name=dyn_skirt_C_1, bone2Name=base_skirt_C_1, swing1LimitDegrees=0, swing2LimitDegrees=20, twistLimitDegrees=10}
 {bone1Name=dyn_skirt_C_2, bone2Name=dyn_skirt_C_1, swing1LimitDegrees=10, swing2LimitDegrees=20, twistLimitDegrees=10}
 {bone1Name=dyn_skirt_D_1, bone2Name=base_skirt_D_1, swing1LimitDegrees=0, swing2LimitDegrees=16, twistLimitDegrees=10}
 {bone1Name=dyn_skirt_D_2, bone2Name=dyn_skirt_D_1, swing1LimitDegrees=10, swing2LimitDegrees=20, twistLimitDegrees=10}
 {bone1Name=dyn_skirt_E_2, bone2Name=dyn_skirt_E_1, swing1LimitDegrees=10, swing2LimitDegrees=20, twistLimitDegrees=10}
 {bone1Name=dyn_skirt_E_1, bone2Name=base_skirt_E_1, swing1LimitDegrees=0, swing2LimitDegrees=18, twistLimitDegrees=10}
 {
  bone1Name=thigh_l
  bone2Name=pelvis
  swing1LimitDegrees=45
  swing1Motion=Free
  swing2LimitDegrees=45
  swing2Motion=Free
  twistLimitDegrees=45
  twistMotion=Free
 }
 {
  bone1Name=base_skirt_A_1
  bone2Name=pelvis
  swing1LimitDegrees=45
  swing1Motion=Free
  swing2LimitDegrees=45
  swing2Motion=Free
  twistLimitDegrees=45
  twistMotion=Free
 }
 {
  bone1Name=base_skirt_B_1
  bone2Name=pelvis
  swing1LimitDegrees=45
  swing1Motion=Free
  swing2LimitDegrees=45
  swing2Motion=Free
  twistLimitDegrees=45
  twistMotion=Free
 }
 {
  bone1Name=base_skirt_C_1
  bone2Name=pelvis
  swing1LimitDegrees=45
  swing1Motion=Free
  swing2LimitDegrees=45
  swing2Motion=Free
  twistLimitDegrees=45
  twistMotion=Free
 }
 {
  bone1Name=base_skirt_D_1
  bone2Name=pelvis
  swing1LimitDegrees=45
  swing1Motion=Free
  swing2LimitDegrees=45
  swing2Motion=Free
  twistLimitDegrees=45
  twistMotion=Free
 }
 {
  bone1Name=base_skirt_E_1
  bone2Name=pelvis
  swing1LimitDegrees=45
  swing1Motion=Free
  swing2LimitDegrees=45
  swing2Motion=Free
  twistLimitDegrees=45
  twistMotion=Free
 }
 {
  bone1Name=thigh_r
  bone2Name=pelvis
  swing1LimitDegrees=45
  swing1Motion=Free
  swing2LimitDegrees=45
  swing2Motion=Free
  twistLimitDegrees=45
  twistMotion=Free
 }
 {
  bone1Name=kin_knife
  bone2Name=thigh_l
  swing1LimitDegrees=45
  swing1Motion=Free
  swing2LimitDegrees=45
  swing2Motion=Free
  twistLimitDegrees=45
  twistMotion=Free
 }
]
