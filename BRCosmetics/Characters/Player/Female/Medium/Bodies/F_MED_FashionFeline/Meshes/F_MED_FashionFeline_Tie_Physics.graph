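# Physics asset for the FashionFeline medium female body's tie and bell, presumably for secondary motion on the cosmetic.
# Four two-segment chains (top and bottom, left and right) of capsules or boxes and a bell sphere hang from small base bodies with narrow angular limits.
# Base bodies and shoulder and breast collider capsules connect to spine_05 through angularly free constraints.
# A large spine_05 capsule covers the upper torso.
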PhysicsAsset /BRCosmetics/Characters/Player/Female/Medium/Bodies/F_MED_FashionFeline/Meshes/F_MED_FashionFeline_Tie_Physics
bodies {
 base_bell [{shapeType=Capsule, length=0.5, radius=0.25, shapeName=base_tieBot_1_l_capsule}]
 base_tieBot_1_l [{shapeType=Capsule, length=0.5, radius=0.25, shapeName=base_tieBot_1_l_capsule}]
 base_tieBot_1_r [{shapeType=Capsule, length=0.5, radius=0.25, shapeName=base_tieBot_1_r_capsule}]
 base_tieTop_1_l [{shapeType=Capsule, length=0.5, radius=0.25, shapeName=base_tieTop_1_l_capsule}]
 base_tieTop_1_r [{shapeType=Capsule, length=0.5, radius=0.25, shapeName=base_tieTop_1_r_capsule}]
 coll_breast_l [{shapeType=Capsule, length=10.4, radius=1.86, shapeName=coll_shoulder_l_capsule, center=<4 0 -0.3187>, rotation=<rot 90 0 0>}]
 coll_breast_r [{shapeType=Capsule, length=10.4, radius=1.86, shapeName=coll_shoulder_l_capsule, center=<4 0 -0.1258>, rotation=<rot 90 0 0>}]
 coll_shoulder_l [{shapeType=Capsule, length=10.4, radius=1.86, shapeName=coll_shoulder_l_capsule, center=<4 0 -0.3187>, rotation=<rot 90 0 0>}]
 coll_shoulder_r [{shapeType=Capsule, length=10.4, radius=1.86, shapeName=coll_shoulder_r_capsule, center=<4 0 -0.3187>, rotation=<rot 90 180 0>}]
 dyn_bell [{shapeType=Sphere, radius=2.4454, shapeName=dyn_bell_sphere, center=<2.3428 0 -0.5515>}]
 dyn_tieBot_1_l [{
   shapeType=Box
   extentX=4.0817
   extentY=0.6843
   extentZ=1.7278
   shapeName=dyn_tieBot_1_l_box
   center=<2.4925 0 0>
   rotation=<rot 0 -0.0001 -12.9979>
  }]
 dyn_tieBot_1_r [{shapeType=Box, extentX=4.0817, extentY=0.6843, extentZ=2.0767, shapeName=dyn_tieBot_1_r_box, center=<2.4925 0 0>}]
 dyn_tieBot_2_l [{
   shapeType=Box
   extentX=4.8243
   extentY=0.6843
   extentZ=2.663
   shapeName=dyn_tieBot_2_l_box
   center=<2.385 0 0>
   rotation=<rot 0 -0.0002 -11.3521>
  }]
 dyn_tieBot_2_r [{shapeType=Box, extentX=4.8243, extentY=0.6843, extentZ=2.663, shapeName=dyn_tieBot_2_r_box, center=<2.385 0 0>}]
 dyn_tieTop_1_l [{
   shapeType=Capsule
   length=2.5427
   radius=1.0424
   shapeName=dyn_tieTop_1_l_capsule
   center=<3.0171 -0.3834 -0.2867>
   rotation=<rot -14.1368 -89.0399 -5.2016>
  }]
 dyn_tieTop_1_r [{
   shapeType=Capsule
   length=2.5427
   radius=1.0424
   shapeName=dyn_tieTop_1_l_capsule
   center=<3.0171 -0.3834 -0.2867>
   rotation=<rot -14.1368 -89.0399 -5.2016>
  }]
 dyn_tieTop_2_l [{
   shapeType=Capsule
   length=3.3265
   radius=1.579
   shapeName=dyn_tieTop_2_l_capsule
   center=<1.8673 -0.357 0.0343>
   rotation=<rot -13.5745 -90.3194 3.1456>
  }]
 dyn_tieTop_2_r [{
   shapeType=Capsule
   length=3.1607
   radius=1.2466
   shapeName=dyn_tieTop_2_l_capsule
   center=<1.8706 -0.4406 0.0838>
   rotation=<rot 14.1963 -88.8063 -8.1306>
  }]
 spine_05 [{
   shapeType=Capsule
   length=10.3699
   radius=10.7147
   shapeName=spine_05_capsule
   center=<-4.2364 1.0098 0.2525>
   rotation=<rot -5.5643 -90 1.4045>
  }]
}
constraints [
 {bone1Name=dyn_bell, bone2Name=base_bell, swing1LimitDegrees=10, swing2LimitDegrees=20, twistLimitDegrees=0}
 {bone1Name=dyn_tieBot_1_l, bone2Name=base_tieBot_1_l, swing1LimitDegrees=15, swing2LimitDegrees=15, twistLimitDegrees=10}
 {bone1Name=dyn_tieBot_2_l, bone2Name=dyn_tieBot_1_l, swing1LimitDegrees=15, swing2LimitDegrees=15, twistLimitDegrees=10}
 {bone1Name=dyn_tieBot_1_r, bone2Name=base_tieBot_1_r, swing1LimitDegrees=15, swing2LimitDegrees=15, twistLimitDegrees=10}
 {bone1Name=dyn_tieBot_2_r, bone2Name=dyn_tieBot_1_r, swing1LimitDegrees=15, swing2LimitDegrees=15, twistLimitDegrees=10}
 {bone1Name=dyn_tieTop_1_l, bone2Name=base_tieTop_1_l, swing1LimitDegrees=15, swing2LimitDegrees=0, twistLimitDegrees=0}
 {bone1Name=dyn_tieTop_2_l, bone2Name=dyn_tieTop_1_l, swing1LimitDegrees=15, swing2LimitDegrees=15, twistLimitDegrees=10}
 {bone1Name=dyn_tieTop_1_r, bone2Name=base_tieTop_1_r, swing1LimitDegrees=15, swing2LimitDegrees=0, twistLimitDegrees=0}
 {bone1Name=dyn_tieTop_2_r, bone2Name=dyn_tieTop_1_r, swing1LimitDegrees=15, swing2LimitDegrees=15, twistLimitDegrees=10}
 {
  bone1Name=base_tieTop_1_l
  bone2Name=spine_05
  swing1LimitDegrees=45
  swing1Motion=Free
  swing2LimitDegrees=45
  swing2Motion=Free
  twistLimitDegrees=45
  twistMotion=Free
 }
 {
  bone1Name=base_tieTop_1_r
  bone2Name=spine_05
  swing1LimitDegrees=45
  swing1Motion=Free
  swing2LimitDegrees=45
  swing2Motion=Free
  twistLimitDegrees=45
  twistMotion=Free
 }
 {
  bone1Name=base_tieBot_1_r
  bone2Name=spine_05
  swing1LimitDegrees=45
  swing1Motion=Free
  swing2LimitDegrees=45
  swing2Motion=Free
  twistLimitDegrees=45
  twistMotion=Free
 }
 {
  bone1Name=base_tieBot_1_l
  bone2Name=spine_05
  swing1LimitDegrees=45
  swing1Motion=Free
  swing2LimitDegrees=45
  swing2Motion=Free
  twistLimitDegrees=45
  twistMotion=Free
 }
 {
  bone1Name=base_bell
  bone2Name=spine_05
  swing1LimitDegrees=45
  swing1Motion=Free
  swing2LimitDegrees=45
  swing2Motion=Free
  twistLimitDegrees=45
  twistMotion=Free
 }
 {
  bone1Name=coll_shoulder_l
  bone2Name=spine_05
  swing1LimitDegrees=45
  swing1Motion=Free
  swing2LimitDegrees=45
  swing2Motion=Free
  twistLimitDegrees=45
  twistMotion=Free
 }
 {
  bone1Name=coll_shoulder_r
  bone2Name=spine_05
  swing1LimitDegrees=45
  swing1Motion=Free
  swing2LimitDegrees=45
  swing2Motion=Free
  twistLimitDegrees=45
  twistMotion=Free
 }
 {
  bone1Name=coll_breast_r
  bone2Name=spine_05
  swing1LimitDegrees=45
  swing1Motion=Free
  swing2LimitDegrees=45
  swing2Motion=Free
  twistLimitDegrees=45
  twistMotion=Free
 }
 {
  bone1Name=coll_breast_l
  bone2Name=spine_05
  swing1LimitDegrees=45
  swing1Motion=Free
  swing2LimitDegrees=45
  swing2Motion=Free
  twistLimitDegrees=45
  twistMotion=Free
 }
]
