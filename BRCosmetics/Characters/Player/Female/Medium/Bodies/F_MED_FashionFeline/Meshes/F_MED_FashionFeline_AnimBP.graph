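# Animation Blueprint for the FashionFeline female cosmetic body, layering secondary motion onto the pose copied from its attached parent.
# Cached poses chain Control Rig, bone adjustments, dynamics and post-sim skirt/bag fixes.
# AnimDynamics drives accessories such as bag, toy, zippers, bracelet and bottle; RigidBody nodes simulate skirt and tie.
# Events set gravity, movement-based skirt alpha, montage and glider states, front-end sim-space settings, and hide the tail under capes.
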AnimBlueprint /BRCosmetics/Characters/Player/Female/Medium/Bodies/F_MED_FashionFeline/Meshes/F_MED_FashionFeline_AnimBP
target_skeleton=&/F_MED_FashionFeline_Skeleton
graphs [
 AnimGraph {nodes [
   Output Pose(Result=Use cached pose 'postSimFixes'()) {layerGroup=None}
   n1 = Control Rig!(Source=Copy Pose From Mesh() {bUseAttachedParent=true}, Alpha, pub isRigidBodiesEnabled:bool=Get bRigidBodyEnabled()) -> (Pose) {
    alphaBoolBlend {blendOption=Linear}
    controlRigAssetReference {blueprintRigClass=&/F_MED_FashionFeline_CtrlRig.F_MED_FashionFeline_CtrlRig_C}
   }
   Save cached pose 'controlRigCP'(Pose=@n1) {cacheName=controlRigCP}
   n2 = Anim Dynamics - Bone: dyn_bag!(
    ComponentPose=Local To Component(LocalPose=Use cached pose 'generalAdjustmentsCP'())
    Alpha=1
    GravityOverride:Vector="0,0,-980"
    SimSpaceSettings:Anim Phys Sim Space Settings Structure=(SimSpaceAngularAlpha=0,MaxAngularVelocity=10000,MaxAngularAcceleration=10000,ExternalAngularVelocity=(X=0,Y=0,Z=0))
   ) -> (Pose) {
    angularDampingOverride=1
    angularSpringConstant=1000
    bAngularSpring=true
    bOverrideAngularDamping=true
    bUseGravityOverride=true
    bUseSphericalLimits=true
    gravityScale=0.35
    simulationSpace=RootRelative
    alphaBoolBlend {blendOption=Linear}
    boundBone {boneName=dyn_bag}
    componentAppliedLinearAccClamp=<4000 4000 4000>
    componentLinearAccScale=<0.4 0.4 0.4>
    componentLinearVelScale=<0.4 0.4 0.4>
    gravityOverride {z=-980}
    physicsBodyDefinitions [{
      collisionType=CustomSphere
      sphereCollisionRadius=6.95
      boundBone {boneName=dyn_bag}
      boxExtents=<20 20 20>
      constraintSetup {angularLimitsMax=<0 15 5>, angularLimitsMin=<0 -15 -15>, angularTarget=<1 0 0>}
      localJointOffset=<10.8 -6.1 0>
     }]
    sphericalLimits [{limitRadius=10.1185, drivingBone {boneName=thigh_r}, sphereLocalOffset=<13.3338 -0.0852 -2.0351>}]
   }
   n3 = Anim Dynamics - Bone: dyn_toyHead!(
    ComponentPose=@n2
    Alpha=1
    GravityOverride:Vector="0,0,-490"
    SimSpaceSettings:Anim Phys Sim Space Settings Structure=(SimSpaceAngularAlpha=0,MaxAngularVelocity=10000,MaxAngularAcceleration=10000,ExternalAngularVelocity=(X=0,Y=0,Z=0))
   ) -> (Pose) {
    angularDampingOverride=0.9999
    angularSpringConstant=5000
    bAngularSpring=true
    bOverrideAngularDamping=true
    bUseGravityOverride=true
    gravityScale=0.35
    simulationSpace=RootRelative
    alphaBoolBlend {blendOption=Linear}
    boundBone {boneName=dyn_toyHead}
    componentAppliedLinearAccClamp=<4000 4000 4000>
    componentLinearAccScale=<1 1 1>
    componentLinearVelScale=<1 1 1>
    gravityOverride {z=-490}
    physicsBodyDefinitions [{
      boundBone {boneName=dyn_toyHead}
      boxExtents=<20 20 20>
      constraintSetup {angularLimitsMax=<0 10 10>, angularLimitsMin=<0 -10 -10>, angularTarget=<1 0 0>}
      localJointOffset=<2 0 0>
     }]
   }
   n4 = Anim Dynamics - Bone: dyn_toyEar_l!(ComponentPose=@n3, Alpha=1, GravityOverride:Vector="0,0,-490") -> (Pose) {
    angularDampingOverride=0.9999
    angularSpringConstant=2000
    bAngularSpring=true
    bOverrideAngularDamping=true
    bUseGravityOverride=true
    gravityScale=0.35
    simulationSpace=RootRelative
    alphaBoolBlend {blendOption=Linear}
    boundBone {boneName=dyn_toyEar_l}
    componentAppliedLinearAccClamp=<4000 4000 4000>
    componentLinearAccScale=<0.5 0.5 0.5>
    componentLinearVelScale=<0.5 0.5 0.5>
    gravityOverride {z=-490}
    physicsBodyDefinitions [{
      boundBone {boneName=dyn_toyEar_l}
      boxExtents=<20 20 20>
      constraintSetup {angularLimitsMax=<0 20 10>, angularLimitsMin=<0 -20 -10>, angularTarget=<1 0 0>}
      localJointOffset=<6 0 0>
     }]
   }
   n5 = Anim Dynamics - Bone: dyn_toyEar_r!(ComponentPose=@n4, Alpha=1, GravityOverride:Vector="0,0,-490") -> (Pose) {
    angularDampingOverride=0.9999
    angularSpringConstant=2000
    bAngularSpring=true
    bOverrideAngularDamping=true
    bUseGravityOverride=true
    gravityScale=0.35
    simulationSpace=RootRelative
    alphaBoolBlend {blendOption=Linear}
    boundBone {boneName=dyn_toyEar_r}
    componentAppliedLinearAccClamp=<4000 4000 4000>
    componentLinearAccScale=<0.5 0.5 0.5>
    componentLinearVelScale=<0.5 0.5 0.5>
    gravityOverride {z=-490}
    physicsBodyDefinitions [{
      boundBone {boneName=dyn_toyEar_r}
      boxExtents=<20 20 20>
      constraintSetup {angularLimitsMax=<0 20 10>, angularLimitsMin=<0 -20 -10>, angularTarget=<1 0 0>}
      localJointOffset=<6 0 0>
     }]
   }
   n6 = Anim Dynamics - Bone: dyn_bagBeltTip!(ComponentPose=@n5, Alpha=1, GravityOverride:Vector="0,0,-980") -> (Pose) {
    angularDampingOverride=1
    angularSpringConstant=1000
    bAngularSpring=true
    bOverrideAngularDamping=true
    bUseGravityOverride=true
    gravityScale=0.35
    simulationSpace=RootRelative
    alphaBoolBlend {blendOption=Linear}
    boundBone {boneName=dyn_bagBeltTip}
    componentAppliedLinearAccClamp=<3000 3000>
    componentLinearAccScale=<1 1 1>
    componentLinearVelScale=<1 1 1>
    gravityOverride {z=-980}
    physicsBodyDefinitions [{
      boundBone {boneName=dyn_bagBeltTip}
      boxExtents=<30 30 30>
      constraintSetup {angularLimitsMax=<0 15 25>, angularLimitsMin=<0 -15 -25>, angularTarget=<1 0 0>}
      localJointOffset=<10 0 0>
     }]
   }
   n7 = Anim Dynamics - Bone: dyn_bootZipper_l!(
    ComponentPose=@n6
    Alpha=1
    SimSpaceSettings:Anim Phys Sim Space Settings Structure=(SimSpaceAngularAlpha=0,MaxAngularVelocity=10000,MaxAngularAcceleration=10000,ExternalAngularVelocity=(X=0,Y=0,Z=0))
   ) -> (Pose) {
    angularDampingOverride=1
    angularSpringConstant=1000
    bAngularSpring=true
    bOverrideAngularDamping=true
    bUseGravityOverride=true
    gravityScale=0.35
    simulationSpace=RootRelative
    alphaBoolBlend {blendOption=Linear}
    boundBone {boneName=dyn_bootZipper_l}
    componentAppliedLinearAccClamp=<5000 5000 50000>
    componentLinearAccScale=<0.5 0.5 0.5>
    componentLinearVelScale=<0.5 0.5 0.5>
    gravityOverride {z=-980}
    physicsBodyDefinitions [{
      boundBone {boneName=dyn_bootZipper_l}
      boxExtents=<10 10 10>
      constraintSetup {angularLimitsMax=<0 10 15>, angularLimitsMin=<0 -30 -15>, angularTarget=<1 0 0>}
      localJointOffset=<5 0 0>
     }]
   }
   n8 = Anim Dynamics - Bone: dyn_bootZipper_r!(
    ComponentPose=@n7
    Alpha=1
    SimSpaceSettings:Anim Phys Sim Space Settings Structure=(SimSpaceAngularAlpha=0,MaxAngularVelocity=10000,MaxAngularAcceleration=10000,ExternalAngularVelocity=(X=0,Y=0,Z=0))
   ) -> (Pose) {
    angularDampingOverride=1
    angularSpringConstant=1000
    bAngularSpring=true
    bOverrideAngularDamping=true
    bUseGravityOverride=true
    gravityScale=0.35
    simulationSpace=RootRelative
    alphaBoolBlend {blendOption=Linear}
    boundBone {boneName=dyn_bootZipper_r}
    componentAppliedLinearAccClamp=<5000 5000 50000>
    componentLinearAccScale=<0.5 0.5 0.5>
    componentLinearVelScale=<0.5 0.5 0.5>
    gravityOverride {z=-980}
    physicsBodyDefinitions [{
      boundBone {boneName=dyn_bootZipper_r}
      boxExtents=<10 10 10>
      constraintSetup {angularLimitsMax=<0 10 15>, angularLimitsMin=<0 -30 -15>, angularTarget=<1 0 0>}
      localJointOffset=<5 0 0>
     }]
   }
   n9 = Anim Dynamics - Bone: dyn_bracelet_r!(ComponentPose=@n8, Alpha=0.7) -> (Pose) {
    angularDampingOverride=1
    angularSpringConstant=800
    bAngularSpring=true
    bLinearSpring=true
    bOverrideLinearDamping=true
    bUseGravityOverride=true
    gravityScale=0
    linearDampingOverride=1
    linearSpringConstant=200
    simulationSpace=RootRelative
    alphaBoolBlend {blendOption=Linear}
    boundBone {boneName=dyn_bracelet_r}
    componentAppliedLinearAccClamp=<3000 3000 3000>
    componentLinearAccScale=<0.5 0.5 0.5>
    componentLinearVelScale=<0.5 0.5 0.5>
    gravityOverride {z=-980}
    physicsBodyDefinitions [{
      boundBone {boneName=dyn_bracelet_r}
      boxExtents=<15 15 15>
      constraintSetup {angularLimitsMax=<0 6 10>, angularLimitsMin=<0 -6 -10>, angularTarget=<1 0 0>, linearAxesMax=<0 0.3 0.2>, linearAxesMin=<0 -0.5 -0.2>}
      localJointOffset=<2 0 0>
     }]
    relativeSpaceBone {boneName=lowerarm_l}
   }
   n10 = Anim Dynamics - Bone: dyn_bottle!(ComponentPose=@n9, Alpha=1) -> (Pose) {
    angularDampingOverride=1
    angularSpringConstant=5000
    bAngularSpring=true
    bOverrideAngularDamping=true
    bUseGravityOverride=true
    gravityScale=0.35
    simulationSpace=RootRelative
    alphaBoolBlend {blendOption=Linear}
    boundBone {boneName=dyn_bottle}
    componentAppliedLinearAccClamp=<3000 3000 4000>
    componentLinearAccScale=<1 1 0.5>
    componentLinearVelScale=<1 1 0.5>
    gravityOverride {z=-980}
    physicsBodyDefinitions [{
      boundBone {boneName=dyn_bottle}
      boxExtents=<20 20 20>
      constraintSetup {angularLimitsMax=<0 2 0>, angularLimitsMin=<0 -2 -20>, angularTarget=<1 0 0>}
      localJointOffset=<5 0 0>
     }]
   }
   n11 = RigidBody!(ComponentPose=@n10, Alpha, OverrideWorldGravity:Vector="0,0,0") -> (Pose) {
    bOverrideWorldGravity=true
    overridePhysicsAsset=&/F_MED_FashionFeline_Skirt_Physics
    simulationSpace=BaseBoneSpace
    alphaBoolBlend {blendOption=Linear}
    alphaScaleBiasClamp {bInterpResult=true, interpSpeedDecreasing=5, interpSpeedIncreasing=5}
    baseBoneRef {boneName=thigh_l}
    componentAppliedLinearAccClamp=<3000 3000 3000>
    overrideWorldGravity {z=-980}
    simSpaceSettings {
     dampingAlpha=0
     maxAngularAcceleration=1
     maxAngularVelocity=3
     maxLinearAcceleration=1000
     maxLinearVelocity=500
     worldAlpha=1
     externalLinearDragV=<0.2 0.2 0.2>
    }
   }
   n12 = RigidBody(ComponentPose=@n11, OverrideWorldGravity="0,0,0") {
    bOverrideWorldGravity=true
    overridePhysicsAsset=&/F_MED_FashionFeline_Tie_Physics
    simulationSpace=BaseBoneSpace
    alphaBoolBlend {blendOption=Linear}
    baseBoneRef {boneName=spine_05}
    componentAppliedLinearAccClamp=<3000 3000 3000>
    overrideWorldGravity {z=-980}
    simSpaceSettings {
     maxAngularAcceleration=1
     maxAngularVelocity=3
     maxLinearAcceleration=1000
     maxLinearVelocity=500
     worldAlpha=1
     externalLinearDragV=<0.2 0.2 0.2>
    }
   }
   Save cached pose 'dynamicsCP'(Pose=Component To Local(ComponentPose=@n12)) {cacheName=dynamicsCP}
   n13 = "Constraint - attach_cape (P)"(ComponentPose=Local To Component(LocalPose=Use cached pose 'controlRigCP'()), Alpha=1, ConstraintWeights_0=1) {
    alphaBoolBlend {blendOption=Linear}
    boneToModify {boneName=attach_cape}
    constraintSetup [{targetBone {boneName=spine_05}}]
    constraintWeights [1]
   }
   n14 = "Transform (Modify) Bone - Bone: attach_cape"!(ComponentPose=@n13, bAlphaBoolEnabled:bool=True, Translation="0,-1,0", Rotation="0,0,16") -> (Pose) {
    alphaInputType=Bool
    rotationMode=BMM_Additive
    rotationSpace=BCS_BoneSpace
    translationMode=BMM_Additive
    translationSpace=BCS_BoneSpace
    alphaBoolBlend {blendInTime=0.5, blendOption=CircularOut, blendOutTime=0.5}
    boneToModify {boneName=attach_cape}
    rotation {roll=16}
    translation {y=-1}
   }
   n15 = "Transform (Modify) Bone - Bone: attach_cape"(ComponentPose=@n14, bAlphaBoolEnabled=True, Translation="0,-2,0", Rotation="0,0,20") {
    alphaInputType=Bool
    rotationMode=BMM_Additive
    rotationSpace=BCS_BoneSpace
    translationMode=BMM_Additive
    translationSpace=BCS_BoneSpace
    alphaBoolBlend {blendInTime=0.5, blendOption=CircularOut, blendOutTime=0.5}
    boneToModify {boneName=attach_cape}
    rotation {roll=20}
    translation {y=-2}
   }
   n16 = "Transform (Modify) Bone - Bone: attach_cape"(ComponentPose=@n15, bAlphaBoolEnabled=True, Translation="0,0,0", Rotation="0,0,10") {
    alphaInputType=Bool
    rotationMode=BMM_Additive
    rotationSpace=BCS_BoneSpace
    translationMode=BMM_Additive
    translationSpace=BCS_BoneSpace
    alphaBoolBlend {blendInTime=0.5, blendOption=Cubic, blendOutTime=0.5}
    boneToModify {boneName=attach_cape}
    rotation {roll=10}
   }
   n17 = "Transform (Modify) Bone - Bone: base_bottle"!(ComponentPose=@n16, Alpha=1, Translation="1,0,0", Rotation="-2,-20,0") -> (Pose) {
    rotationMode=BMM_Additive
    rotationSpace=BCS_BoneSpace
    scaleMode=BMM_Additive
    scaleSpace=BCS_BoneSpace
    translationMode=BMM_Additive
    translationSpace=BCS_BoneSpace
    alphaBoolBlend {blendOption=Cubic}
    boneToModify {boneName=base_bottle}
    rotation {pitch=-2, yaw=-20}
    translation {x=1}
   }
   n18 = "Transform (Modify) Bone - Bone: base_bag"!(ComponentPose=@n17, Alpha=1, Translation="0,0.5,-3", Rotation="0,0,-10") -> (Pose) {
    rotationMode=BMM_Additive
    rotationSpace=BCS_BoneSpace
    scaleMode=BMM_Additive
    scaleSpace=BCS_BoneSpace
    translationMode=BMM_Additive
    translationSpace=BCS_BoneSpace
    alphaBoolBlend {blendOption=Cubic}
    boneToModify {boneName=base_bag}
    rotation {roll=-10}
    translation {y=0.5, z=-3}
   }
   n19 = "Transform (Modify) Bone - Bone: upperarm_r"!(ComponentPose=@n18, Alpha=1, Translation="0,0,0", Rotation="0,-10,0") -> (Pose) {
    rotationMode=BMM_Additive
    rotationSpace=BCS_BoneSpace
    scaleMode=BMM_Additive
    scaleSpace=BCS_BoneSpace
    translationMode=BMM_Additive
    translationSpace=BCS_BoneSpace
    alphaBoolBlend {blendOption=Cubic}
    boneToModify {boneName=upperarm_r}
    rotation {yaw=-10}
   }
   Save cached pose 'generalAdjustmentsCP'(Pose=Component To Local(ComponentPose=@n19)) {cacheName=generalAdjustmentsCP}
   n20 = "Transform (Modify) Bone - Bone: dyn_skirt_A_2"!(
    ComponentPose=Local To Component(LocalPose=Use cached pose 'dynamicsCP'())
    bAlphaBoolEnabled:bool=True
    Translation="0,0,0"
    Rotation="-20,-20,20"
   ) -> (Pose) {
    alphaInputType=Bool
    rotationMode=BMM_Additive
    rotationSpace=BCS_BoneSpace
    translationMode=BMM_Additive
    translationSpace=BCS_BoneSpace
    alphaBoolBlend {blendInTime=1, blendOption=Cubic, blendOutTime=1}
    boneToModify {boneName=dyn_skirt_A_2}
    rotation=<rot -20 -20 20>
   }
   n21 = "Transform (Modify) Bone - Bone: base_skirt_C_1"!(ComponentPose=@n20, bAlphaBoolEnabled:bool=True, Translation="0,0,0", Rotation="70,30,40") -> (Pose) {
    alphaInputType=Bool
    rotationMode=BMM_Additive
    rotationSpace=BCS_BoneSpace
    translationMode=BMM_Additive
    translationSpace=BCS_BoneSpace
    alphaBoolBlend {blendInTime=0.5, blendOption=Cubic, blendOutTime=0.5}
    boneToModify {boneName=base_skirt_C_1}
    rotation=<rot 70 30 40>
   }
   n22 = "Transform (Modify) Bone - Bone: base_skirt_D_1"!(ComponentPose=@n21, bAlphaBoolEnabled:bool=True, Translation="-1.7,0,0", Rotation="64,0,0") -> (Pose) {
    alphaInputType=Bool
    rotationMode=BMM_Additive
    rotationSpace=BCS_BoneSpace
    translationMode=BMM_Additive
    translationSpace=BCS_BoneSpace
    alphaBoolBlend {blendInTime=0.5, blendOption=Cubic, blendOutTime=0.5}
    boneToModify {boneName=base_skirt_D_1}
    rotation {pitch=64}
    translation {x=-1.7}
   }
   n23 = "Constraint - dyn_bag (T)"!(ComponentPose=@n22, bAlphaBoolEnabled:bool=True, ConstraintWeights_0=1) -> (Pose) {
    alphaInputType=Bool
    alphaBoolBlend {blendInTime=0.5, blendOption=Linear, blendOutTime=0.5}
    boneToModify {boneName=dyn_bag}
    constraintSetup [{transformType=Translation, targetBone {boneName=base_bag}}]
    constraintWeights [1]
   }
   Save cached pose 'postSimFixes'(Pose=Component To Local(ComponentPose=@n23)) {cacheName=postSimFixes}
  ]}
 EventGraph {nodes [
   Event BlueprintUpdateAnimation()
   > n1 = Sequence!(execute) -> (then_0, then_1, then_2:Exec, then_3:Exec)
   > n2 = "Gravity Override State Function (Parameter Asset) (Nativized)"(
    CharacterPartAnimInstance=Self-Reference()
    GameTimeInSeconds=GetGameTimeInSeconds()
    Parameters=&/F_MED_FashionFeline_Parameters
   )
   > Set GravityOverride(GravityOverride=@n2)
   n3 = Get bIsJumping()
   n4 = Get bIsTacticalSprinting()
   n5 = Get bIsFalling()
   n6 = Branch(
    execute=@n1.then_1
    Condition=OR Boolean!(A=AND Boolean(A=@n3, B=@n4), B=AND Boolean(A=@n4, B=@n5), C:bool=Get bIsTacticalSprinting(), D:bool=Get bIsKneeSliding()) -> (ReturnValue)
   )
   > Set alphaSkirtRigidBody float(alphaSkirtRigidBody float:double=0.5) -> (Output_Get:double)
   n7 = Branch(execute=@n6.else, Condition=Get bIsBaseballSliding())
   > Set alphaSkirtRigidBody float(alphaSkirtRigidBody float=0)
   n8 = Branch(execute=@n7.else, Condition=OR Boolean!(A=Get bIsJumping(), B=Get bIsFalling()) -> (ReturnValue))
   > Set alphaSkirtRigidBody float(alphaSkirtRigidBody float=0.15)
   n9 = Get bIsCrouching()
   n10 = Branch(execute=@n8.else, Condition=AND Boolean(A=Get bIsSprinting(), B=NOT Boolean(A=@n9)))
   > Set alphaSkirtRigidBody float(alphaSkirtRigidBody float=0.55)
   Set alphaSkirtRigidBody float(execute=@n10.else, alphaSkirtRigidBody float=1)
   n11 = Get bIsMoving2D()
   Set isCrouchingStatic(execute=@n1.then_2, isCrouchingStatic:bool=AND Boolean(A=NOT Boolean(A=@n11), B=Get bIsCrouching())) -> (Output_Get:bool)
   n12 = GetAnimInstance(self=GetOwnerSkeletalMeshComponent())
   n13 = Set isIdleMontage(
    execute=@n1.then_3
    isIdleMontage:double=float * float!(
     A=GetCurveValue(self=@n12, CurveName=DisablePelvisAdjustment)
     B:bool=Montage_IsPlaying(self=@n12, Montage=/BRCosmetics/Animation/Game/MainPlayer/Menu/BR/Female_LobbyIdle_HandsDownWide_Montage)
     C:bool=Get bIsInFrontEnd()
    ) -> (ReturnValue) {is=title K2Node_PromotableOperator}
   ) -> (Output_Get:double)
   > Set isIdleMontageReverse(isIdleMontageReverse:double=Remap Float(Input=@n13, OldRange=(X=0,Y=1), NewRange=(X=1,Y=0), Clamp=true) -> (Output=~)) -> (Output_Get:double)
   Event OnParachuteOpened()
   > Switch on EGliderType(Selection=Get PlayerGliderType() -> (PlayerGliderType=HangGlider))
   > Set isUsingHangGlidingInParachute(isUsingHangGlidingInParachute:bool=true) -> (Output_Get:bool)
   Event OnParachuteClosed()
   > Set isUsingHangGlidingInParachute(isUsingHangGlidingInParachute=false)
   n14 = Event OnCharacterCustomizationLoaded()
   > Set tailMesh(tailMesh:Skeletal Mesh Component Object Reference=Get Array Item with EFort Custom Part Type(Array=@n14.CustomMeshParts, PartType=MiscOrTail)) -> (Output_Get:Skeletal Mesh Component Object Reference)
   > n15 = Sequence!(execute) -> (then_0, then_1, then_2)
   n16 = CheckForBackbling()
   > Set isWearingCape(isWearingCape=OR Boolean(A=@n16.bIsBackblingACape, B=@n16.bIsBackblingLong))
   > Set characterHasTail(characterHasTail:bool=CheckForTail!(self:Custom Character Part Anim Instance Object Reference) -> (bHasTail:bool)) -> (Output_Get:bool)
   Branch(execute=@n15.then_1, Condition=Get characterHasTail() -> (characterHasTail:bool))
   > n17 = Branch(Condition=Get isWearingCape() -> (isWearingCape:bool))
   n18 = Get tailMesh() -> (tailMesh:Skeletal Mesh Component Object Reference)
   > SetVisibility!(execute, self:Scene Component Object Reference=@n18, bNewVisibility:bool=false, bPropagateToChildren:bool=false) -> (then)
   SetVisibility(execute=@n17.else, self=@n18, bNewVisibility=true, bPropagateToChildren=false)
   n19 = Branch(execute=@n15.then_2, Condition=Get bIsInFrontEnd())
   > Set SimSpaceSettings output(SimSpaceSettings output:Anim Phys Sim Space Settings Structure=Get SimSpaceSettings frontEnd() -> (SimSpaceSettings frontEnd:Anim Phys Sim Space Settings Structure)) -> (Output_Get:Anim Phys Sim Space Settings Structure)
   Set SimSpaceSettings output(
    execute=@n19.else
    SimSpaceSettings output=Get SimSpaceSettings default() -> (SimSpaceSettings default:Anim Phys Sim Space Settings Structure)
   )
  ]}
]
variables [
 alphaSkirtRigidBody float {type=number}
 characterHasTail {type=boolean}
 GravityOverride {type=Vector}
 isCrouchingStatic {type=boolean}
 isIdleMontage {type=number}
 isIdleMontageReverse {type=number}
 isUsingHangGlidingInParachute {type=boolean}
 isWearingCape {type=boolean}
 SimSpaceSettings default {
  type=AnimPhysSimSpaceSettings
  default {maxAngularAcceleration=10000, maxAngularVelocity=10000, simSpaceAngularAlpha=0, externalAngularVelocity=<0 0 0>}
 }
 SimSpaceSettings frontEnd {
  type=AnimPhysSimSpaceSettings
  default {maxAngularAcceleration=30, maxAngularVelocity=6, simSpaceAngularAlpha=0.3, externalAngularVelocity=<0 0 0>}
 }
 SimSpaceSettings output {
  type=AnimPhysSimSpaceSettings
  default {maxAngularAcceleration=10000, maxAngularVelocity=10000, simSpaceAngularAlpha=0, externalAngularVelocity=<0 0 0>}
 }
 tailMesh {type=SkeletalMeshComponent}
]
