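# Physics asset for a face-accessory part of the female medium FashionFeline player cosmetic, providing capsule bodies for a dangling earring and a hair bone.
# The earring is a chain from base_earring_1 through three dyn_earring bodies, linked by constraints with narrow swing limits and zero twist limit.
# A larger, offset capsule covers dyn_hair_2_B_l_varA, which has no constraint tying it to the chain.
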PhysicsAsset /BRCosmetics/Characters/Player/Female/Medium/Bodies/F_MED_FashionFeline/Meshes/Parts/F_MED_FashionFeline_FaceAcc_backPack_Physics
bodies {
 base_earring_1 [{shapeType=Capsule, length=0.5, radius=0.25, shapeName=base_earring_1_capsule}]
 dyn_earring_1 [{shapeType=Capsule, length=0.505, radius=0.505, shapeName=dyn_earring_2_capsule, center=<0.9035 0 0>, rotation=<rot 90 0 0>}]
 dyn_earring_2 [{shapeType=Capsule, length=0.505, radius=0.505, shapeName=dyn_earring_2_capsule, center=<0.784 0 0>, rotation=<rot 90 0 0>}]
 dyn_earring_3 [{shapeType=Capsule, length=3.0574, radius=0.7933, shapeName=dyn_earring_3_capsule, center=<2.1646 0 0>, rotation=<rot 90 0 0>}]
 dyn_hair_2_B_l_varA [{
   shapeType=Capsule
   length=7.8448
   radius=1.1766
   shapeName=dyn_hair_2_B_l_varA_capsule
   center=<3.3312 -0.0499 0.3263>
   rotation=<rot 0.0205 98.3003 -89.6809>
  }]
}
constraints [
 {bone1Name=dyn_earring_1, bone2Name=base_earring_1, swing1LimitDegrees=15, swing2LimitDegrees=15, twistLimitDegrees=0}
 {bone1Name=dyn_earring_2, bone2Name=dyn_earring_1, swing1LimitDegrees=20, swing2LimitDegrees=20, twistLimitDegrees=0}
 {bone1Name=dyn_earring_3, bone2Name=dyn_earring_2, swing1LimitDegrees=20, swing2LimitDegrees=20, twistLimitDegrees=0}
]
